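# Niagara system for the SlickSwish skydiving contrail cosmetic, drawing pink-magenta electricity on the character.
# Limb lightning meshes jitter within cylinders around sampled bones, with flickering color and material masks for traveling or static bolts.
# Socket-attached electricity meshes scale with the contrail setup's dive value and add glow parameters.
# Time of day and front-end preview adjust the look; some older or variant emitters are disabled.
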NiagaraSystem /BRCosmetics/Athena/Effects/Particles/Skydiving/NS_Contrail_SlickSwish/NS_Contrail_SlickSwish
dependencies {
 usedDataInterfaces [
  &NiagaraDataInterfaceArrayColor
  &NiagaraDataInterfaceArrayFloat3
  &NiagaraDataInterfaceCurve
  &NiagaraDataInterfaceArrayFloat
  &NiagaraDataInterfaceMeshRendererInfo
  /Script/Niagara.NiagaraDataInterfaceArrayQuat
  &NiagaraDataInterfaceArrayFloat2
  &NiagaraDataInterfaceVector2DCurve
  &NiagaraDataInterfaceArrayInt32
 ]
 usedDynamicInputs [&MaskIntBySpawnGroup]
 usedModules [
  &self:SystemSpawnScript.SystemScriptSource.SystemScriptGraph.NiagaraNodeAssignment_0.SetVariables_#1
  &self:SystemSpawnScript.SystemScriptSource.SystemScriptGraph.NiagaraNodeAssignment_1.SetVariables_#2
  &SystemState
  /Game/Effects/Niagara/Modules/Contrails/FN_Contrail_Setup
  &FN_Integrator
  &self:SystemSpawnScript.SystemScriptSource.SystemScriptGraph.NiagaraNodeAssignment_2.SetVariables_#3
 ]
}
emitters {
 Arms-Legs_Lightning {
  emitterUpdateScript [
   &EmitterState()
   &SpawnRate(SpawnRate=25, Use Spawn Probability=false)
   /SetVariables(Emitter.DepthFade=35) {isSetParameters=true, module=SetVariables_#4}
  ]
  particleSpawnScript [
   &InitializeParticle(
    Color=&FNTimeOfDayColor(
     Night Color=<rgba 50 0.7521 25.1492 1>
     Morning Color=<rgba 50 0.7521 25.1492 1>
     Day Color=<rgba 50 0.7521 25.1492 1>
     Evening Color=<rgba 50 0.7521 25.1492 1>
    )
    Lifetime=&RandomRangeFloat(Minimum=0.5, Maximum=1)
    Mesh Scale=&Multiply_VectorByFloat(
     Vector=&SelectVectorFromArray(
      Array Sampling Mode=::NewEnumerator0
      Direct Array Index=?
      Vector Selection Array=DataInterface() {propertyValues {FloatData [<1 -1 1>, <-1 -1 -1>]}}
      Random Seed=0
     )
     Float=&RandomRangeFloat(Minimum=0.85, Maximum=1.25)
    )
    Mesh Scale Mode=::NewEnumerator4
    UsePositionOffset=false
    Write Material Random=false
   )
   /SetVariables(Particles.BoneIndex:NiagaraInt32=&SelectIntFromArray(
     Array Sampling Mode=::NewEnumerator1
     Direct Array Index=?
     Int Selection Array=DataInterface() {propertyValues {IntData [0,1,2,3,5,6,7,8]}}
     Random Seed=0
    )) {isSetParameters=true, module=SetVariables_#5}
   /SetVariables(
    Particles.MeshIndex=&RandomRangeInt(Minimum=0, Maximum=5)
    Particles.bIsTravelingBolt:NiagaraBool=&SetBoolByFloatComparison(
     A=&RandomRangeFloat(Minimum=0.5, Maximum=1)
     B=&FN_FrontEndFloatModifier(Gameplay=0.25, Front End=0)
    )
   ) {isSetParameters=true, module=SetVariables_#6}
   /SetVariables(Particles.MaskSize:NiagaraFloat=&MakeCustomFloatFromBool(Bool=$Particles.bIsTravelingBolt, True Float=&RandomRangeFloat(Minimum=0.75, Maximum=1), False Float=2)) {isSetParameters=true, module=SetVariables_#7}
   &ShapeLocation(
    Apply To Particle Position=::NewEnumerator1
    Cylinder Height=&SelectFloatFromArray(Direct Array Index=$Particles.BoneIndex, Float Selection Array=DataInterface() {propertyValues {FloatData [20,20,20,20,0,20,20,30,30]}})
    Cylinder Height Midpoint=1
    Cylinder Radius=10
    MaskSpawnGroup=false
    Override Seed=false
    Rotation Axis=&SelectVectorFromArray(
     Direct Array Index=$Particles.BoneIndex
     Vector Selection Array=DataInterface() {propertyValues {FloatData [<-1 0 0>, <1 0 0>, <-1 0 0>, <1 0 0>, <0 0 0>, <1 0 0>, <-1 0 0>, <1 0 0>, <-1 0 0>]}}
    )
    Rotation Mode=::NewEnumerator1
    Shape Origin=$Engine.Emitter.SimulationPosition
    Shape Primitive=::NewEnumerator1
    Surface Only=false
    UseOwnerScale=true
   )
   &SkeletalMeshLocation(
    Bone / Socket Index=$Particles.BoneIndex
    Bone Sampling Mode=::NewEnumerator4
    MaskSpawnGroup=false
    Orientation Sampling=::NewEnumerator1
    Sampled Position Offset:NiagaraPosition=$Particles.ShapeLocation.ShapePosition
    Skeletal Mesh=$System.Skeletal Mesh
   ) {enabled=false, module=SkeletalMeshLocation001}
   &DynamicMaterialParameters(
    Index 0 Param 1=$System.LightningWidth
    Index 0 Param 2=$Particles.MaskSize
    Index 0 Param 4=0.5
    Module.Param0WriteEnabled=true
    Module.Param1WriteEnabled=true
    Module.Param2WriteEnabled=true
    Module.Param3WriteEnabled=true
   )
   &CameraOffset(Camera Offset Amount=10)
  ]
  particleUpdateScript [
   &SolveForcesAndVelocity(Clamp Velocity=false, Limit Acceleration=false)
   &ParticleState()
   &ShapeLocation(
    Apply To Particle Position=::NewEnumerator1
    Cylinder Distribution=::NewEnumerator2
    Cylinder Height=&SelectFloatFromArray(Direct Array Index=$Particles.BoneIndex, Float Selection Array=DataInterface() {propertyValues {FloatData [20,20,20,20,0,20,20,30,30]}})
    Cylinder Height Midpoint=1
    Cylinder Radius=2
    MaskSpawnGroup=false
    Rotation Axis=&SelectVectorFromArray(
     Direct Array Index=$Particles.BoneIndex
     Vector Selection Array=DataInterface() {propertyValues {FloatData [<-1 0 0>, <1 0 0>, <-1 0 0>, <1 0 0>, <0 0 0>, <1 0 0>, <-1 0 0>, <1 0 0>, <-1 0 0>]}}
    )
    Rotation Mode=::NewEnumerator1
    Shape Primitive=::NewEnumerator1
    U Position=&RandomRangeFloat(Minimum=0, Maximum=1)
    UseOwnerScale=true
    V Position=&Lerp_Float(
     A=&FloatFromCurve(
      FloatCurve=DataInterface() {propertyValues {Curve {defaultValue=1, keys [{tangentMode=RCTM_User, time=0, value=1}, {tangentMode=RCTM_User, time=1, value=0}]}}}
      Scale Curve=&RandomRangeFloat(Minimum=1, Maximum=2)
     )
     B=&RandomRangeFloat(Minimum=0, Maximum=1)
     Alpha=&SelectFloatFromArray(
      Array Sampling Mode=::NewEnumerator0
      Direct Array Index=?
      Float Selection Array=DataInterface() {propertyValues {FloatData [0.75,1]}}
      Random Seed=0
     )
    )
   ) {module=ShapeLocation001}
   &SkeletalMeshLocation(
    Bone / Socket Index=$Particles.BoneIndex
    Bone Sampling Mode=::NewEnumerator4
    MaskSpawnGroup=false
    Orientation Sampling=::NewEnumerator1
    Sampled Position Offset:NiagaraPosition=$Particles.ShapeLocation001.ShapePosition
    Skeletal Mesh=$System.Skeletal Mesh
   )
   /SetVariables(
    Particles.MaskIntensity:NiagaraFloat=&MakeCustomFloatFromBool(
     Bool=$Particles.bIsTravelingBolt
     True Float=2
     False Float=&FloatFromCurve(FloatCurve=DataInterface() {propertyValues {Curve {keys [{tangentMode=RCTM_User, time=0, value=0.333}, {tangentMode=RCTM_User, time=1, value=0.666}]}}})
    )
    Particles.MaskPosition:NiagaraFloat=&MakeCustomFloatFromBool(
     Bool=$Particles.bIsTravelingBolt
     True Float=$Particles.NormalizedAge
     False Float=&FloatFromCurve(FloatCurve=DataInterface() {propertyValues {Curve {keys [<at 0 0>,<at 0.333 0.5>]}}}, Scale Curve=$System.LightningWidth)
    )
    Particles.MeshPosition:NiagaraPosition=$Particles.Position
   ) {isSetParameters=true, module=SetVariables_#8}
   &DynamicMaterialParameters(
    Index 0 Param 1=&FloatFromCurve(FloatCurve=DataInterface() {propertyValues {Curve {keys [<at 0.666 1>,<at 1 0>]}}}, Scale Curve=$System.LightningWidth)
    Index 0 Param 2=$Particles.MaskSize
    Index 0 Param 3=$Particles.MaskPosition
    Index 0 Param 4=$Particles.MaskIntensity
    Module.Param0WriteEnabled=true
    Module.Param1WriteEnabled=true
    Module.Param2WriteEnabled=true
    Module.Param3WriteEnabled=true
   ) {module=DynamicMaterialParameters001}
   &ScaleColor(
    Scale RGB=&VectorFromFloat(Value=&Multiply_Float(
      A=&Lerp_Float(
       A=0.133
       B=1
       Alpha=&Waveform("Waveform [1]"=::NewEnumerator0, "[1] Frequency"=5, "[1] Amplitude Min/Max"=<0.5 1>)
      )
      B=1
     ))
    ScaleA=false
    ScaleRGB=true
   )
   &ScaleColor(
    Color Value To Scale=$Particles.Color
    Scale RGB=&Multiply_VectorByFloat(Vector=<1 1 1>, Float=$System.TimeOfDayBrightnessMultiplier)
    ScaleA=true
    ScaleRGB=true
   ) {module=ScaleColor001}
   /SetVariables(
    Particles.CameraOffsetGlow=9.5
    Particles.ColorGlow:LinearColor=<rgba 0.401 0 0.1722 5>
    Particles.DynamicMaterialParameterGlow:Vector4f=&MakeVector4(
     X=&Add_Float(A=$System.LightningWidth, B=6)
     Y=$Particles.MaskSize
     Z=$Particles.MaskPosition
     W=$Particles.MaskIntensity
    )
   ) {isSetParameters=true, module=SetVariables_#9}
   /SetVariables(Particles.MeshOrientation=&MultiplyQuaternion(
     Quaternion A=$Particles.SkeletalMeshLocation.SampledRotation
     Quaternion B=/Niagara/DynamicInputs/Arrays/SelectQuaternionFromArray(
      Array Sampling Mode:ENiagaraArraySamplingMode=::NewEnumerator1
      Direct Array Index:NiagaraInt32=$Particles.BoneIndex
      Interpolation:NiagaraFloat=?
      Direct Array Mode:ENiagara_WrapClamp=::NewEnumerator0
      Evaluation Type:ENiagaraRandomnessEvaluation=::NewEnumerator0
      Normalized Interpolation:NiagaraBool=true
      Recalculate Random Each Loop:NiagaraBool=false
      Randomness Mode:ENiagaraRandomnessMode=::NewEnumerator0
      Random Seed:NiagaraInt32=?
      Fixed Random Seed:NiagaraBool=false
      Quaternion Selection Array:NiagaraDataInterfaceArrayQuat=DataInterface() {propertyValues {
        GpuSyncMode=SyncCpuToGpu
        MaxElements=0
        QuatData [
         <0 0 0.7071 -0.7071>
         <0 0 0.7071 0.7071>
         <0 0 -0.7071 0.7071>
         <0 0 0.7071 0.7071>
         <0 0 0 1>
         <0 0 0.7071 0.7071>
         <0 0 -0.7071 0.7071>
         <0 0 0.7071 0.7071>
         <0 0 -0.7071 0.7071>
        ]
       }}
      Enable Spherical Linear Interpolation:NiagaraBool=?
     )
    )) {isSetParameters=true, module=SetVariables_#10}
  ]
  rendererClasses [&NiagaraMeshRendererProperties]
  renderers [
   {rendererClass=&NiagaraMeshRendererProperties, rendererIndex=0}
   {rendererClass=&NiagaraMeshRendererProperties, rendererIndex=1}
  ]
 }
 Arms_Lightning {
  bEnabled=false
  emitterUpdateScript [
   &EmitterState()
   &SpawnRate(SpawnRate=15, Use Spawn Probability=false)
   /SetVariables(Emitter.DepthFade=35) {isSetParameters=true, module=SetVariables_#4}
  ]
  particleSpawnScript [
   &InitializeParticle(
    Color=&FNTimeOfDayColor(
     Night Color=<rgba 50 0.7521 25.1492 1>
     Morning Color=<rgba 50 0.7521 25.1492 1>
     Day Color=<rgba 50 0.7521 25.1492 1>
     Evening Color=<rgba 50 0.7521 25.1492 1>
    )
    Lifetime=&RandomRangeFloat(Minimum=0.5, Maximum=1)
    Mesh Scale=&Multiply_VectorByFloat(
     Vector=&SelectVectorFromArray(
      Array Sampling Mode=::NewEnumerator0
      Direct Array Index=?
      Vector Selection Array=DataInterface() {propertyValues {FloatData [<1 -1 1>, <-1 -1 -1>]}}
      Random Seed=0
     )
     Float=&RandomRangeFloat(Minimum=0.85, Maximum=1.25)
    )
    Mesh Scale Mode=::NewEnumerator4
    UsePositionOffset=false
    Write Material Random=false
   )
   /SetVariables(Particles.BoneIndex:NiagaraInt32=&RandomRangeInt(Minimum=0, Maximum=3)) {isSetParameters=true, module=SetVariables_#5}
   /SetVariables(
    Particles.MeshIndex=&RandomRangeInt(Minimum=0, Maximum=5)
    Particles.bIsTravelingBolt:NiagaraBool=&SetBoolByFloatComparison(
     A=&RandomRangeFloat(Minimum=0.5, Maximum=1)
     B=&FN_FrontEndFloatModifier(Gameplay=0.25, Front End=0)
    )
   ) {isSetParameters=true, module=SetVariables_#6}
   /SetVariables(Particles.MaskSize:NiagaraFloat=&MakeCustomFloatFromBool(Bool=$Particles.bIsTravelingBolt, True Float=&RandomRangeFloat(Minimum=0.75, Maximum=1), False Float=2)) {isSetParameters=true, module=SetVariables_#7}
   &ShapeLocation(
    Apply To Particle Position=::NewEnumerator1
    Cylinder Height=&SelectFloatFromArray(Direct Array Index=$Particles.BoneIndex, Float Selection Array=DataInterface() {propertyValues {FloatData [25,25,50,50]}})
    Cylinder Height Midpoint=1
    Cylinder Radius=10
    MaskSpawnGroup=false
    Override Seed=false
    Rotation Axis=&SelectVectorFromArray(
     Direct Array Index=$Particles.BoneIndex
     Vector Selection Array=DataInterface() {propertyValues {FloatData [<-1 0 0>, <1 0 0>, <-1 0 0>, <1 0 0>]}}
     Random Seed=0
    )
    Rotation Mode=::NewEnumerator1
    Shape Origin=$Engine.Emitter.SimulationPosition
    Shape Primitive=::NewEnumerator1
    Surface Only=false
    UseOwnerScale=true
   )
   &SkeletalMeshLocation(
    Bone / Socket Index=$Particles.BoneIndex
    Bone Sampling Mode=::NewEnumerator4
    MaskSpawnGroup=false
    Orientation Sampling=::NewEnumerator1
    Sampled Position Offset:NiagaraPosition=$Particles.ShapeLocation.ShapePosition
    Skeletal Mesh=$System.Skeletal Mesh
   ) {enabled=false, module=SkeletalMeshLocation001}
   &DynamicMaterialParameters(
    Index 0 Param 1=$System.LightningWidth
    Index 0 Param 2=$Particles.MaskSize
    Index 0 Param 4=0.5
    Module.Param0WriteEnabled=true
    Module.Param1WriteEnabled=true
    Module.Param2WriteEnabled=true
    Module.Param3WriteEnabled=true
   )
   &CameraOffset(Camera Offset Amount=10)
  ]
  particleUpdateScript [
   &SolveForcesAndVelocity(Clamp Velocity=false, Limit Acceleration=false)
   &ParticleState()
   &ShapeLocation(
    Apply To Particle Position=::NewEnumerator1
    Cylinder Distribution=::NewEnumerator2
    Cylinder Height=&SelectFloatFromArray(Direct Array Index=$Particles.BoneIndex, Float Selection Array=DataInterface() {propertyValues {FloatData [20,20,20,20]}})
    Cylinder Height Midpoint=1
    Cylinder Radius=2
    MaskSpawnGroup=false
    Rotation Axis=&SelectVectorFromArray(
     Direct Array Index=$Particles.BoneIndex
     Vector Selection Array=DataInterface() {propertyValues {FloatData [<-1 0 0>, <1 0 0>, <-1 0 0>, <1 0 0>]}}
    )
    Rotation Mode=::NewEnumerator1
    Shape Primitive=::NewEnumerator1
    U Position=&RandomRangeFloat(Minimum=0, Maximum=1)
    UseOwnerScale=true
    V Position=&Lerp_Float(
     A=&FloatFromCurve(
      FloatCurve=DataInterface() {propertyValues {Curve {defaultValue=1, keys [{tangentMode=RCTM_User, time=0, value=1}, {tangentMode=RCTM_User, time=1, value=0}]}}}
      Scale Curve=&RandomRangeFloat(Minimum=1, Maximum=2)
     )
     B=&RandomRangeFloat(Minimum=0, Maximum=1)
     Alpha=&SelectFloatFromArray(
      Array Sampling Mode=::NewEnumerator0
      Direct Array Index=?
      Float Selection Array=DataInterface() {propertyValues {FloatData [0.75,1]}}
      Random Seed=0
     )
    )
   ) {module=ShapeLocation001}
   &SkeletalMeshLocation(
    Bone / Socket Index=$Particles.BoneIndex
    Bone Sampling Mode=::NewEnumerator4
    MaskSpawnGroup=false
    Orientation Sampling=::NewEnumerator1
    Sampled Position Offset:NiagaraPosition=$Particles.ShapeLocation001.ShapePosition
    Skeletal Mesh=$System.Skeletal Mesh
   )
   /SetVariables(
    Particles.MaskIntensity:NiagaraFloat=&MakeCustomFloatFromBool(
     Bool=$Particles.bIsTravelingBolt
     True Float=2
     False Float=&FloatFromCurve(FloatCurve=DataInterface() {propertyValues {Curve {keys [{tangentMode=RCTM_User, time=0, value=0.333}, {tangentMode=RCTM_User, time=1, value=0.666}]}}})
    )
    Particles.MaskPosition:NiagaraFloat=&MakeCustomFloatFromBool(
     Bool=$Particles.bIsTravelingBolt
     True Float=$Particles.NormalizedAge
     False Float=&FloatFromCurve(FloatCurve=DataInterface() {propertyValues {Curve {keys [<at 0 0>,<at 0.333 0.5>]}}}, Scale Curve=$System.LightningWidth)
    )
    Particles.MeshPosition:NiagaraPosition=$Particles.Position
   ) {isSetParameters=true, module=SetVariables_#8}
   &DynamicMaterialParameters(
    Index 0 Param 1=&FloatFromCurve(FloatCurve=DataInterface() {propertyValues {Curve {keys [<at 0.666 1>,<at 1 0>]}}}, Scale Curve=$System.LightningWidth)
    Index 0 Param 2=$Particles.MaskSize
    Index 0 Param 3=$Particles.MaskPosition
    Index 0 Param 4=$Particles.MaskIntensity
    Module.Param0WriteEnabled=true
    Module.Param1WriteEnabled=true
    Module.Param2WriteEnabled=true
    Module.Param3WriteEnabled=true
   ) {module=DynamicMaterialParameters001}
   &ScaleColor(
    Scale RGB=&VectorFromFloat(Value=&Multiply_Float(
      A=&Lerp_Float(
       A=0.133
       B=1
       Alpha=&Waveform("Waveform [1]"=::NewEnumerator0, "[1] Frequency"=5, "[1] Amplitude Min/Max"=<0.5 1>)
      )
      B=1
     ))
    ScaleA=false
    ScaleRGB=true
   )
   &ScaleColor(
    Color Value To Scale=$Particles.Color
    Scale RGB=&Multiply_VectorByFloat(Vector=<1 1 1>, Float=$System.TimeOfDayBrightnessMultiplier)
    ScaleA=true
    ScaleRGB=true
   ) {module=ScaleColor001}
   /SetVariables(
    Particles.CameraOffsetGlow=9.5
    Particles.ColorGlow:LinearColor=<rgba 0.401 0 0.1722 5>
    Particles.DynamicMaterialParameterGlow:Vector4f=&MakeVector4(
     X=&Add_Float(A=$System.LightningWidth, B=6)
     Y=$Particles.MaskSize
     Z=$Particles.MaskPosition
     W=$Particles.MaskIntensity
    )
   ) {isSetParameters=true, module=SetVariables_#9}
   /SetVariables(Particles.MeshOrientation=&MultiplyQuaternion(
     Quaternion A=$Particles.SkeletalMeshLocation.SampledRotation
     Quaternion B=&SelectQuaternionFromArray(
      Array Sampling Mode:ENiagaraArraySamplingMode=::NewEnumerator1
      Direct Array Index:NiagaraInt32=$Particles.BoneIndex
      Interpolation:NiagaraFloat=?
      Direct Array Mode:ENiagara_WrapClamp=::NewEnumerator0
      Evaluation Type:ENiagaraRandomnessEvaluation=::NewEnumerator0
      Normalized Interpolation:NiagaraBool=true
      Recalculate Random Each Loop:NiagaraBool=false
      Randomness Mode:ENiagaraRandomnessMode=::NewEnumerator0
      Random Seed:NiagaraInt32=?
      Fixed Random Seed:NiagaraBool=false
      Quaternion Selection Array:NiagaraDataInterfaceArrayQuat=DataInterface() {propertyValues {GpuSyncMode=SyncCpuToGpu, MaxElements=0, QuatData [<0 0 0.7071 -0.7071>, <0 0 0.7071 0.7071>, <0 0 -0.7071 0.7071>, <0 0 0.7071 0.7071>]}}
      Enable Spherical Linear Interpolation:NiagaraBool=?
     )
    )) {isSetParameters=true, module=SetVariables_#10}
  ]
  rendererClasses [&NiagaraMeshRendererProperties]
  renderers [
   {rendererClass=&NiagaraMeshRendererProperties, rendererIndex=0}
   {rendererClass=&NiagaraMeshRendererProperties, rendererIndex=1}
  ]
 }
 ElectricityMesh {
  emitterUpdateScript [
   &EmitterState(Life Cycle Mode=::NewEnumerator1, Loop Duration=1, UseLoopDelay=false)
   &SpawnRate(SpawnRate=5, Use Spawn Probability=false)
   &SpawnRate(SpawnRate=5, Use Spawn Probability=false) {enabled=false, module=SpawnRate001}
  ]
  particleSpawnScript [
   /SetVariables(Particles.MeshIndex=&RandomRangeInt(Minimum=0, Maximum=2)) {isSetParameters=true}
   &PartitionParticles()
   &InitializeParticle(
    Color=<rgba 5 0.651 2.8379 1>
    Lifetime=&RandomRangeFloat(Minimum=1.25, Maximum=1)
    Mesh Renderer Array Visibility Mode=::NewEnumerator3
    Mesh Renderer Info=DataInterface() {propertyValues {MeshRenderer=&self:ElectricityMesh_0.NiagaraMeshRendererProperties_1}}
    Mesh Scale=&Multiply_Vector(
     A=&RandomRangeVector(Minimum=<0.125 0.125 0.125>, Maximum=<0.15 0.15 0.125>)
     B=&MakeVector(
      X=&SelectFloatFromArray(
       Direct Array Index=$Particles.PartitionParticles.Partition
       Float Selection Array=DataInterface() {propertyValues {FloatData [-1.5,1.5,0.1,0.1]}}
      )
      Y=&SelectFloatFromArray(
       Direct Array Index=$Particles.PartitionParticles.Partition
       Float Selection Array=DataInterface() {propertyValues {FloatData [-1.5,1.5,0.1,0.1]}}
      )
      Z=&Lerp_Float(
       A=0.7
       B=1.1
       Alpha=&FloatFromCurve(
        FloatCurve=DataInterface() {propertyValues {Curve {keys [
            {arriveTangent=4.1785, interpMode=RCIM_Cubic, time=0, value=0}
            {arriveTangent=1, interpMode=RCIM_Cubic, leaveTangent=1, time=0.5, value=0.2}
            {interpMode=RCIM_Cubic, leaveTangent=0.0224, time=1, value=1}
           ]}}}
        CurveIndex=10
       )
      )
     )
    )
    Mesh Scale Mode=::NewEnumerator4
    Sprite Rotation Mode=::NewEnumerator2
    Sprite Size Mode=::NewEnumerator3
    UsePositionOffset=false
    Write Material Random=true
   )
   &SkeletalMeshLocation(
    Bone / Socket Index=$Particles.PartitionParticles.Partition
    Bone Sampling Mode=::NewEnumerator4
    Interpolate Mesh Sampling=false
    MaskSpawnGroup=false
    Orientation Sampling=::NewEnumerator1
    Skeletal Mesh=$System.Skeletal Mesh
   ) {enabled=false, module=SkeletalMeshLocation001}
   &InitialMeshOrientation(Mesh Axis To Orient=<0 0 -1>, Mesh Orientation Mode=::NewEnumerator2, Use Rotation Vector=true)
   /SetVariables(Particles.UVy_Pos:NiagaraFloat=&RandomRangeFloat(Minimum=0, Maximum=22)) {isSetParameters=true}
   /SetVariables(Particles.MeshOrientation=/Niagara/DynamicInputs/TypeConversions/MakeCustomQuaternionFromBool(
     Bool:NiagaraBool=$User.bIsFrontEndPreview
     True Quat:Quat4f=&MakeQuaternion(Quaternion From=2, Axis=<1 0 0>, Angle=0, Y=45, Z=-90, From=?, To=?, Angle Type=::NewEnumerator0, Normalize Vectors=false)
     False Quat:Quat4f=<0 -1 0 1>
    )) {enabled=false, isSetParameters=true}
  ]
  particleUpdateScript [
   &ParticleState()
   /SetVariables(
    Particles.LightningLength:NiagaraFloat=&FloatFromCurve(
     FloatCurve=DataInterface() {propertyValues {Curve {
        defaultValue=0.9
        keys [{arriveTangent=1.0525, interpMode=RCIM_Cubic, leaveTangent=1.0525, tangentMode=RCTM_User, time=0, value=0.5}, <at 1 1 RCIM_Cubic>]
       }}}
     Scale Curve=&Multiply_Float(
      A=&Lerp_Float(
       A=1
       B=&SelectFloatFromArray(Direct Array Index=$Particles.PartitionParticles.Partition, Float Selection Array=DataInterface() {propertyValues {FloatData [1.4,1.4,1,1]}})
       Alpha=$System.FN_Contrail_Setup.Dive Value
      )
      B=&SelectFloatFromArray(Direct Array Index=$Particles.PartitionParticles.Partition, Float Selection Array=DataInterface() {propertyValues {FloatData [1.5,1.5,1,1]}})
     )
    )
    Particles.LightningShapeExpand:NiagaraFloat=&FloatFromCurve(
     FloatCurve=DataInterface() {propertyValues {Curve {
        defaultValue=0
        keys [
         {arriveTangent=4.1785, interpMode=RCIM_Cubic, leaveTangent=4.1785, tangentMode=RCTM_User, time=0, value=0}
         {arriveTangent=1.1179, interpMode=RCIM_Cubic, leaveTangent=1.1179, tangentMode=RCTM_User, time=0.2042, value=0.6228}
         {arriveTangent=0.0224, interpMode=RCIM_Cubic, leaveTangent=0.0224, tangentMode=RCTM_User, time=1, value=1}
        ]
       }}}
     Scale Curve=&Lerp_Float(
      A=0.5
      B=&SelectFloatFromArray(
       Direct Array Index=$Particles.PartitionParticles.Partition
       Float Selection Array=DataInterface() {propertyValues {FloatData [0.35,0.35,0.5,0.5]}}
      )
      Alpha=$System.FN_Contrail_Setup.Dive Value
     )
    )
   ) {isSetParameters=true, module=SetVariables_#11}
   &ScaleMeshSize(Scale Factor=&MakeVector(
     X=$Particles.LightningShapeExpand
     Y=$Particles.LightningShapeExpand
     Z=&Multiply_Float(A=$Particles.LightningLength, B=&Lerp_Float(A=0.8, B=1.25, Alpha=0))
    ))
   &SkeletalMeshLocation(
    Bone / Socket Index=$Particles.PartitionParticles.Partition
    MaskSpawnGroup=false
    Mesh Sampling Type=::NewEnumerator5
    Orientation Sampling=::NewEnumerator1
    Sampled Position Offset=&SelectVectorFromArray(
     Direct Array Index=$Particles.PartitionParticles.Partition
     Vector Selection Array=DataInterface() {propertyValues {FloatData [<10 0 5>, <10 0 -5>, <10 0 0>, <10 0 0>]}}
    )
    Skeletal Mesh=$System.Skeletal Mesh
    Socket Sampling Mode=::NewEnumerator5
   )
   &DynamicMaterialParameters(
    Index 0 Param 1=&FloatFromCurve(FloatCurve=DataInterface() {propertyValues {Curve {keys [<at 0 1>,<at 1 0.5>]}}}, Scale Curve=1.5)
    Index 0 Param 3=$Particles.UVy_Pos
    Index 0 Param 4=0.5
    Index 1 Param 1=&FloatFromCurve(
     FloatCurve=DataInterface() {propertyValues {Curve {keys [
         <at 0 0>
         {arriveTangent=1, interpMode=RCIM_Cubic, leaveTangent=1, time=0.3, value=0}
         {interpMode=RCIM_Cubic, leaveTangent=-8.0089, time=1, value=1}
        ]}}}
     Scale Curve=0.1
    )
    Index 1 Param 2=1
    Index 1 Param 3=&FloatFromCurve(FloatCurve=DataInterface() {propertyValues {Curve {keys [<at 0.5 1 RCIM_Cubic>,<at 1 0 RCIM_Cubic>]}}})
    Index 2 Param 1=&FloatFromCurve(
     FloatCurve=DataInterface() {propertyValues {Curve {keys [{tangentMode=RCTM_User, time=0, value=0}, {tangentMode=RCTM_User, time=1, value=1}]}}}
     Scale Curve=0
    )
    Index 2 Param 2=2
    Index 3 Param 4=30
    Module.Index0Float0Write=true
    Module.Index0Float1Write=true
    Module.Index0Float2Write=true
    Module.Index0Float3Write=true
    Module.Index1Float0Write=true
    Module.Index1Float1Write=true
    Module.Index1Float2Write=true
    Module.Index1Float3Write=true
    Module.Index2Float0Write=true
    Module.Index2Float1Write=true
    Module.Index2Float2Write=true
    Module.Index2Float3Write=true
    Module.Index3Float0Write=true
    Module.Index3Float1Write=true
    Module.Index3Float2Write=true
    Module.Index3Float3Write=true
    Write Parameter Index 1=::NewEnumerator1
    Write Parameter Index 2=::NewEnumerator1
    Write Parameter Index 3=::NewEnumerator1
   )
   &ScaleColor(
    Scale Alpha=&FloatFromCurve(FloatCurve=DataInterface() {propertyValues {Curve {keys [
         {tangentMode=RCTM_User, time=0, value=1}
         {arriveTangent=-0.0405, interpMode=RCIM_Cubic, leaveTangent=-0.0405, tangentMode=RCTM_User, time=0.7009, value=1}
         {arriveTangent=-8.0089, interpMode=RCIM_Cubic, leaveTangent=-8.0089, tangentMode=RCTM_User, time=1, value=0}
        ]}}})
    Scale RGB=&VectorFromFloat(Value=&FloatFromCurve(
      FloatCurve=DataInterface() {propertyValues {Curve {keys [
          {arriveTangent=9.0725, interpMode=RCIM_Cubic, time=0, value=0}
          <at 0.5033 2 RCIM_Cubic>
          {interpMode=RCIM_Cubic, leaveTangent=-0.0405, time=0.9112, value=0}
         ]}}}
      Scale Curve=&FN_FrontEndFloatModifier(Gameplay=3, Front End=3)
     ))
    ScaleA=true
    ScaleRGB=true
   ) {module=ScaleColor001}
   &ScaleColor(
    Scale Alpha=&FloatFromCurve(FloatCurve=DataInterface() {propertyValues {Curve {keys [
         {tangentMode=RCTM_User, time=0, value=1}
         {arriveTangent=-0.0405, interpMode=RCIM_Cubic, leaveTangent=-0.0405, tangentMode=RCTM_User, time=0.7009, value=1}
         {arriveTangent=-8.0089, interpMode=RCIM_Cubic, leaveTangent=-8.0089, tangentMode=RCTM_User, time=1, value=0}
        ]}}})
    Scale RGB=&VectorFromFloat(Value=&FloatFromCurve(
      FloatCurve=DataInterface() {propertyValues {Curve {keys [
          {arriveTangent=9.0725, interpMode=RCIM_Cubic, time=0, value=0}
          <at 0.1408 2 RCIM_Cubic>
          {interpMode=RCIM_Cubic, leaveTangent=-0.0405, time=1, value=0}
         ]}}}
      Scale Curve=&FN_FrontEndFloatModifier(Gameplay=10, Front End=8)
     ))
    ScaleA=true
    ScaleRGB=true
   ) {enabled=false}
   /SetVariables(Particles.MeshOrientation=&MakeQuaternion(
     Quaternion From=2
     Axis=<1 0 0>
     Angle=0
     Y=&RandomRangeFloat(Minimum=75, Maximum=105)
     Z=&RandomRangeFloat(Minimum=0, Maximum=360)
     From=?
     To=?
     Angle Type=::NewEnumerator0
     Normalize Vectors=false
    )) {enabled=false, isSetParameters=true, module=SetVariables_#12}
   /SetVariables(Particles.MeshOrientation=&MakeCustomQuaternionFromBool(
     Bool:NiagaraBool=$User.bIsFrontEndPreview
     True Quat:Quat4f=&LerpQuaternion(
      Quaternion A:Quat4f=<0 -1 -0.2 1>
      Quaternion B:Quat4f=&MakeQuaternion(Quaternion From=2, Axis=<1 0 0>, Angle=0, Y=45, Z=-90, From=?, To=?, Angle Type=::NewEnumerator0, Normalize Vectors=false)
      Lerp Factor:NiagaraFloat=$System.FN_Contrail_Setup.Dive Value
      Enable Spherical Linear Interpolation:NiagaraBool=false
     )
     False Quat:Quat4f=<0 -1 0 1>
    )) {isSetParameters=true}
   &CameraOffset(Camera Offset Amount=1)
   /SetVariables(
    Particles.GlowColor=&MultiplyLinearColor(
     Linear Color=<rgba 0.25 0 0.1074 100>
     Scale Factor=&MakeVector4(
      X=1
      Y=1
      Z=1
      W=&FloatFromCurve(FloatCurve=DataInterface() {propertyValues {Curve {keys [
           {arriveTangent=9.0725, interpMode=RCIM_Cubic, time=0, value=0}
           {arriveTangent=2.4467, interpMode=RCIM_Cubic, leaveTangent=2.4467, time=0.5033, value=2}
           {arriveTangent=-1.1256, interpMode=RCIM_Cubic, leaveTangent=-1.1256, tangentMode=RCTM_User, time=0.8006, value=1.9588}
           {interpMode=RCIM_Cubic, leaveTangent=-0.0405, time=0.9112, value=0}
          ]}}})
     )
    )
    Particles.GlowDynamic:Vector4f=&MakeVector4(
     X=&FloatFromCurve(FloatCurve=DataInterface() {propertyValues {Curve {keys [<at 0 1>,<at 0.75 0.5>]}}}, Scale Curve=1.7)
     Y=0
     Z=$Particles.UVy_Pos
     W=0.5
    )
   ) {isSetParameters=true}
   /SetVariables(
    Particles.GlowColor2:LinearColor=&MultiplyLinearColor(
     Linear Color=<rgba 0.8958 0 0.3847 5>
     Scale Factor=&MakeVector4(
      X=1
      Y=1
      Z=1
      W=&FloatFromCurve(FloatCurve=DataInterface() {propertyValues {Curve {keys [
           {arriveTangent=9.0725, interpMode=RCIM_Cubic, time=0, value=0}
           {arriveTangent=2.4467, interpMode=RCIM_Cubic, leaveTangent=2.4467, time=0.5033, value=2}
           {arriveTangent=-1.1256, interpMode=RCIM_Cubic, leaveTangent=-1.1256, tangentMode=RCTM_User, time=0.8006, value=1.9588}
           {interpMode=RCIM_Cubic, leaveTangent=-0.0405, time=0.9112, value=0}
          ]}}})
     )
    )
    Particles.GlowDynamic2:Vector4f=&MakeVector4(
     X=&FloatFromCurve(FloatCurve=DataInterface() {propertyValues {Curve {keys [<at 0 1>,<at 0.75 0.5>]}}}, Scale Curve=1.6)
     Y=0
     Z=$Particles.UVy_Pos
     W=0.5
    )
   ) {isSetParameters=true}
   /SetVariables(
    Particles.GlowColor=&MultiplyLinearColor(
     Linear Color=<rgba 0.4 0 0.1718 100>
     Scale Factor=&MakeVector4(
      X=1
      Y=1
      Z=1
      W=&FloatFromCurve(FloatCurve=DataInterface() {propertyValues {Curve {keys [
           {arriveTangent=9.0725, interpMode=RCIM_Cubic, time=0, value=0}
           {arriveTangent=2.4467, interpMode=RCIM_Cubic, leaveTangent=2.4467, time=0.5033, value=2}
           {arriveTangent=-1.1256, interpMode=RCIM_Cubic, leaveTangent=-1.1256, tangentMode=RCTM_User, time=0.8006, value=1.9588}
           {interpMode=RCIM_Cubic, leaveTangent=-0.0405, time=0.9112, value=0}
          ]}}})
     )
    )
    Particles.GlowDynamic:Vector4f=&MakeVector4(
     X=&FloatFromCurve(FloatCurve=DataInterface() {propertyValues {Curve {keys [<at 0 1>,<at 0.75 0.5>]}}}, Scale Curve=1.5)
     Y=0
     Z=$Particles.UVy_Pos
     W=0.5
    )
   ) {enabled=false, isSetParameters=true}
   /SetVariables(Particles.GlowOffset=0.5) {isSetParameters=true}
   /SetVariables(
    Particles.Color=&Multiply_LinearColorByFloat(
     Linear Color=$Particles.Color
     Float=&FNTimeOfDayScalar(Night Scale=0.15, Morning Scale=0.25, Day Scale=0.3, Evening Scale=0.25)
    )
    Particles.GlowColor=&MultiplyLinearColor(
     Linear Color=$Particles.GlowColor
     Scale Factor=&MakeVector4(
      X=1
      Y=1
      Z=1
      W=&FNTimeOfDayScalar(Night Scale=0.25, Morning Scale=0.25, Day Scale=0.25, Evening Scale=0.25)
     )
    )
   ) {isSetParameters=true}
  ]
  rendererClasses [&NiagaraMeshRendererProperties]
  renderers [
   {rendererClass=&NiagaraMeshRendererProperties, rendererIndex=0}
   {rendererClass=&NiagaraMeshRendererProperties, rendererIndex=1}
   {rendererClass=&NiagaraMeshRendererProperties, rendererIndex=2}
  ]
 }
 ElectricityOLD {
  bEnabled=false
  emitterUpdateScript [
   &EmitterState(Life Cycle Mode=::NewEnumerator1, Loop Duration=1, UseLoopDelay=false)
   &SpawnRate(SpawnRate=14, Use Spawn Probability=false)
   &SpawnRate(SpawnRate=5, Use Spawn Probability=false) {enabled=false, module=SpawnRate001}
  ]
  particleSpawnScript [
   &InitializeParticle(
    Color=<rgba 4 0.1208 2.2285 1>
    Lifetime=&RandomRangeFloat(Minimum=0.25, Maximum=0.75)
    Mesh Renderer Array Visibility Mode=::NewEnumerator3
    Mesh Renderer Info=DataInterface() {propertyValues {MeshRenderer=&self:ElectricityOLD_0.NiagaraMeshRendererProperties_1}}
    Mesh Scale=&Multiply_Vector(
     A=&RandomRangeVector(Minimum=<0.125 0.125 0.125>, Maximum=<0.15 0.15 0.175>)
     B=&MakeVector(
      X=&SelectFloatFromArray(
       Array Sampling Mode=::NewEnumerator0
       Direct Array Index=?
       Float Selection Array=DataInterface() {propertyValues {FloatData [-1.5,1.5]}}
       Random Seed=0
      )
      Y=&SelectFloatFromArray(
       Array Sampling Mode=::NewEnumerator0
       Direct Array Index=?
       Float Selection Array=DataInterface() {propertyValues {FloatData [-1.5,1.5]}}
       Random Seed=0
      )
      Z=&Lerp_Float(
       A=0.7
       B=1.1
       Alpha=&FloatFromCurve(
        FloatCurve=DataInterface() {propertyValues {Curve {keys [
            {arriveTangent=4.1785, interpMode=RCIM_Cubic, time=0, value=0}
            {arriveTangent=1, interpMode=RCIM_Cubic, leaveTangent=1, time=0.5, value=0.2}
            {interpMode=RCIM_Cubic, leaveTangent=0.0224, time=1, value=1}
           ]}}}
        CurveIndex=10
       )
      )
     )
    )
    Mesh Scale Mode=::NewEnumerator4
    Sprite Rotation Mode=::NewEnumerator2
    Sprite Size Mode=::NewEnumerator3
    UsePositionOffset=false
    Write Material Random=true
   )
   /SetVariables(Particles.VisibilityTag=$Particles.SpawnGroup) {isSetParameters=true}
   &PartitionParticles(Number Of Partitions=4)
   &SkeletalMeshLocation(
    Bone / Socket Index=$Particles.PartitionParticles.Partition
    Bone Sampling Mode=::NewEnumerator4
    Interpolate Mesh Sampling=false
    MaskSpawnGroup=false
    Orientation Sampling=::NewEnumerator1
    Skeletal Mesh=$System.Skeletal Mesh
   ) {enabled=false, module=SkeletalMeshLocation001}
   &InitialMeshOrientation(Mesh Axis To Orient=<0 0 -1>, Mesh Orientation Mode=::NewEnumerator2, Use Rotation Vector=true)
  ]
  particleUpdateScript [
   &ParticleState()
   /SetVariables(
    Particles.LightningLength:NiagaraFloat=&FloatFromCurve(
     FloatCurve=DataInterface() {propertyValues {Curve {
        defaultValue=0.9
        keys [{arriveTangent=1.0525, interpMode=RCIM_Cubic, leaveTangent=1.0525, tangentMode=RCTM_User, time=0, value=0.5}, <at 1 1 RCIM_Cubic>]
       }}}
     Scale Curve=&Multiply_Float(
      A=&Lerp_Float(
       A=1
       B=&SelectFloatFromArray(
        Direct Array Index=$Particles.PartitionParticles.Partition
        Float Selection Array=DataInterface() {propertyValues {FloatData [1.4,1.4,0.5,0.5]}}
       )
       Alpha=$System.FN_Contrail_Setup.Dive Value
      )
      B=&SelectFloatFromArray(Direct Array Index=$Particles.PartitionParticles.Partition, Float Selection Array=DataInterface() {propertyValues {FloatData [1.5,1.5,1,1]}})
     )
    )
    Particles.LightningShapeExpand:NiagaraFloat=&FloatFromCurve(
     FloatCurve=DataInterface() {propertyValues {Curve {
        defaultValue=0
        keys [
         {arriveTangent=4.1785, interpMode=RCIM_Cubic, leaveTangent=4.1785, tangentMode=RCTM_User, time=0, value=0}
         {arriveTangent=1.1179, interpMode=RCIM_Cubic, leaveTangent=1.1179, tangentMode=RCTM_User, time=0.2042, value=0.6228}
         {arriveTangent=0.0224, interpMode=RCIM_Cubic, leaveTangent=0.0224, tangentMode=RCTM_User, time=1, value=1}
        ]
       }}}
     Scale Curve=&Lerp_Float(
      A=0.5
      B=&SelectFloatFromArray(
       Direct Array Index=$Particles.PartitionParticles.Partition
       Float Selection Array=DataInterface() {propertyValues {FloatData [0.35,0.35,0.1,0.1]}}
      )
      Alpha=$System.FN_Contrail_Setup.Dive Value
     )
    )
   ) {isSetParameters=true, module=SetVariables_#11}
   &ScaleMeshSize(Scale Factor=&MakeVector(
     X=$Particles.LightningShapeExpand
     Y=$Particles.LightningShapeExpand
     Z=&Multiply_Float(A=$Particles.LightningLength, B=&Lerp_Float(A=0.8, B=1.25, Alpha=$System.Audio_DrumKick))
    ))
   &UpdateMeshOrientation(
    Facing Direction=&Add_Vector(
     A=<1 0 0>
     B=&Add_Vector(
      A=&Lerp_Vector(A=<0 0 0>, B=<0 0 0.2>, Alpha=$System.Timeline.Time)
      B=&Lerp_Vector(A=<0 0 0>, B=<0 0 -0.2>, Alpha=$System.Timeline001.Time)
     )
    )
    Facing Interpolation=false
    Orientation Method=::NewEnumerator1
    Override Quat=false
    TransformFacing=false
    TransformUp=false
   ) {enabled=false}
   &SkeletalMeshLocation(
    Bone / Socket Index=$Particles.PartitionParticles.Partition
    MaskSpawnGroup=false
    Mesh Sampling Type=::NewEnumerator5
    Orientation Sampling=::NewEnumerator1
    Sampled Position Offset=&SelectVectorFromArray(
     Direct Array Index=$Particles.PartitionParticles.Partition
     Vector Selection Array=DataInterface() {propertyValues {FloatData [<10 0 0>, <10 0 0>, <15 0 0>, <15 0 0>]}}
    )
    Skeletal Mesh=$System.Skeletal Mesh
    Socket Sampling Mode=::NewEnumerator5
   )
   /SetVariables(Particles.OrientationVec:Vector3f=/Niagara/DynamicInputs/Lerp/SmoothLerpOverTime_Vector(
     Convergence Rate:NiagaraFloat=50
     Smooth Value:Vector3f=$Particles.OrientationVec
     Target Value:Vector3f=&Lerp_Vector(A=<0 0 -1>, B=<0 0 -1>, Alpha=$System.Dive)
    )) {enabled=false, isSetParameters=true}
   /SetVariables(Particles.OrientationVec:Vector3f=&self:FindElectricityDirection(
     AudioIntensity:NiagaraFloat=$System.AudioIntensity
     DiveIntensity:NiagaraFloat=$System.Dive
     SourcePosition:NiagaraPosition=$Particles.Position
     TargetPosition:NiagaraPosition=$System.SkullPosition
    )) {enabled=false, isSetParameters=true}
   &UpdateMeshOrientation(
    Facing Coordinate Space:ENiagaraCoordinateSpace=::World
    Facing Interpolation=false
    Orientation Method=::NewEnumerator1
    Override Quat=false
    TransformFacing=true
    TransformUp=true
    Up Direction:Vector3f=<0 0 0>
   ) {enabled=false, module=UpdateMeshOrientation001}
   &DynamicMaterialParameters(
    Index 0 Param 1=&FloatFromCurve(FloatCurve=DataInterface() {propertyValues {Curve {keys [<at 0 1>,<at 1 0.5>]}}}, Scale Curve=1.5)
    Index 0 Param 3=&RandomRangeFloat(Minimum=0, Maximum=22)
    Index 0 Param 4=0.5
    Index 1 Param 1=&FloatFromCurve(
     FloatCurve=DataInterface() {propertyValues {Curve {keys [
         <at 0 0>
         {arriveTangent=1, interpMode=RCIM_Cubic, leaveTangent=1, time=0.3, value=0}
         {interpMode=RCIM_Cubic, leaveTangent=-8.0089, time=1, value=1}
        ]}}}
     Scale Curve=0.1
    )
    Index 1 Param 2=1
    Index 1 Param 3=&FloatFromCurve(FloatCurve=DataInterface() {propertyValues {Curve {keys [<at 0.5 1 RCIM_Cubic>,<at 1 0 RCIM_Cubic>]}}}, Scale Curve=2)
    Index 2 Param 1=&FloatFromCurve(
     FloatCurve=DataInterface() {propertyValues {Curve {keys [{tangentMode=RCTM_User, time=0, value=0}, {tangentMode=RCTM_User, time=1, value=1}]}}}
     Scale Curve=0
    )
    Index 2 Param 2=2
    Index 3 Param 4=30
    Module.Index0Float0Write=true
    Module.Index0Float1Write=true
    Module.Index0Float2Write=true
    Module.Index0Float3Write=true
    Module.Index1Float0Write=true
    Module.Index1Float1Write=true
    Module.Index1Float2Write=true
    Module.Index1Float3Write=true
    Module.Index2Float0Write=true
    Module.Index2Float1Write=true
    Module.Index2Float2Write=true
    Module.Index2Float3Write=true
    Module.Index3Float0Write=true
    Module.Index3Float1Write=true
    Module.Index3Float2Write=true
    Module.Index3Float3Write=true
    Write Parameter Index 1=::NewEnumerator1
    Write Parameter Index 2=::NewEnumerator1
    Write Parameter Index 3=::NewEnumerator1
   )
   &ScaleColor(
    Scale Alpha=&FloatFromCurve(FloatCurve=DataInterface() {propertyValues {Curve {keys [
         {tangentMode=RCTM_User, time=0, value=1}
         {arriveTangent=-0.0405, interpMode=RCIM_Cubic, leaveTangent=-0.0405, tangentMode=RCTM_User, time=0.7009, value=1}
         {arriveTangent=-8.0089, interpMode=RCIM_Cubic, leaveTangent=-8.0089, tangentMode=RCTM_User, time=1, value=0}
        ]}}})
    Scale RGB=&VectorFromFloat(Value=&FloatFromCurve(
      FloatCurve=DataInterface() {propertyValues {Curve {keys [
          {arriveTangent=9.0725, interpMode=RCIM_Cubic, time=0, value=0}
          <at 0.1408 2 RCIM_Cubic>
          {interpMode=RCIM_Cubic, leaveTangent=-0.0405, time=1, value=0}
         ]}}}
      Scale Curve=&FN_FrontEndFloatModifier(Gameplay=10, Front End=8)
     ))
    ScaleA=true
    ScaleRGB=true
   ) {module=ScaleColor001}
   &SkeletalMeshLocation(
    Bone / Socket Index=4
    Bone Sampling Mode=::NewEnumerator4
    MaskSpawnGroup=false
    Orientation Sampling=::NewEnumerator1
    Position Sampling=::NewEnumerator2
    Skeletal Mesh=$System.Skeletal Mesh
    Use Sampled Mesh Attribute as Ribbon ID=false
   ) {enabled=false, module=SkeletalMeshLocation002}
   /SetVariables(Particles.GlideLeanQuat:Quat4f=&MultiplyQuaternion(
     Quaternion A=&MultiplyQuaternion(
      Quaternion A=$Particles.Initial.MeshOrientation
      Quaternion B=&MultiplyQuaternion(
       Quaternion A=&MakeQuatFromAxisAngle(Axis=<1 0 0>, AngleInDegrees=&Lerp_Float(A=0, B=90, Alpha=$System.Timeline.Time))
       Quaternion B=&MakeQuatFromAxisAngle(Axis=<1 0 0>, AngleInDegrees=&Lerp_Float(A=0, B=-90, Alpha=$System.Timeline001.Time))
      )
     )
     Quaternion B=&MakeQuatFromAxisAngle(
      Axis=<0 0 1>
      AngleInDegrees=&Lerp_Float(
       A=0
       B=-80
       Alpha=&RemapRange(Input Value To Remap=$System.Dive, Input Min=0.28, Input Max=1, Output Min=0, Output Max=1)
      )
     )
    )) {enabled=false, isSetParameters=true}
   /SetVariables(Particles.MeshOrientation=&MakeQuaternion(
     Quaternion From=2
     Axis=<1 0 0>
     Angle=0
     Y=&RandomRangeFloat(Minimum=75, Maximum=105)
     Z=&RandomRangeFloat(Minimum=0, Maximum=360)
     From=?
     To=?
     Angle Type=::NewEnumerator0
     Normalize Vectors=false
    )) {isSetParameters=true, module=SetVariables_#12}
   &CameraOffset(Camera Offset Amount=15)
  ]
  rendererClasses [&NiagaraMeshRendererProperties]
  renderers [{rendererClass=&NiagaraMeshRendererProperties, rendererIndex=0}]
 }
 ElectricitySprite {
  emitterUpdateScript [
   &EmitterState()
   &SpawnRate(SpawnRate=7.5, Use Spawn Probability=false)
   &SpawnBurst_Instantaneous(Spawn Count=4, Spawn Probability=HlslExpression() {hlslExpression="Emitter.LoopCount == 0"}, Use Spawn Probability=true) {enabled=false}
  ]
  particleSpawnScript [
   &PartitionParticles()
   &InitializeParticle(
    Color=<rgba 5 0.651 2.8379 1>
    Lifetime=0.5
    Sprite Size=&SelectVector2DFromArray(
     Array Sampling Mode=::NewEnumerator0
     Direct Array Index=?
     Vector 2D Selection Array=DataInterface() {propertyValues {FloatData [<15 100>, <-15 100>]}}
     Random Seed=0
    )
    Sprite Size Mode=::NewEnumerator4
    UsePositionOffset=false
    Write Material Random=true
   )
   /SetVariables(Particles.UVy_Pos:NiagaraFloat=&RandomRangeFloat(Minimum=0, Maximum=22)) {isSetParameters=true, module=SetVariables_#13}
  ]
  particleUpdateScript [
   &ParticleState()
   &ScaleSpriteSize(
    Scale Factor=&Lerp_Vector2(
     A=&SelectVector2DFromArray(
      Direct Array Index=$Particles.PartitionParticles.Partition
      Vector 2D Selection Array=DataInterface() {propertyValues {FloatData [<1 1.5>, <1 1.5>, <1 1>, <1 1>]}}
     )
     B=&SelectVector2DFromArray(
      Direct Array Index=$Particles.PartitionParticles.Partition
      Vector 2D Selection Array=DataInterface() {propertyValues {FloatData [<1 2.25>, <1 2.25>, <1 1.25>, <1 1.25>]}}
     )
     Alpha=$System.FN_Contrail_Setup.Dive Value
    )
    Scale Sprite Size Mode=::NewEnumerator2
   )
   &SkeletalMeshLocation(
    Bone / Socket Index=$Particles.PartitionParticles.Partition
    MaskSpawnGroup=false
    Mesh Sampling Type=::NewEnumerator5
    Orientation Sampling=::NewEnumerator1
    Sampled Position Offset=&SelectVectorFromArray(
     Direct Array Index=$Particles.PartitionParticles.Partition
     Vector Selection Array=DataInterface() {propertyValues {FloatData [<10 0 5>, <10 0 -5>, <10 0 0>, <10 0 0>]}}
    )
    Skeletal Mesh=$System.Skeletal Mesh
    Socket Sampling Mode=::NewEnumerator5
   )
   &DynamicMaterialParameters(
    Index 0 Param 1=&FloatFromCurve(FloatCurve=DataInterface() {propertyValues {Curve {keys [<at 0 1>,<at 1 0.5>]}}}, Scale Curve=0.4)
    Index 0 Param 3=$Particles.UVy_Pos
    Index 0 Param 4=0.5
    Index 1 Param 1=&FloatFromCurve(
     FloatCurve=DataInterface() {propertyValues {Curve {keys [
         <at 0 0>
         {arriveTangent=1, interpMode=RCIM_Cubic, leaveTangent=1, time=0.3, value=0}
         {interpMode=RCIM_Cubic, leaveTangent=-8.0089, time=1, value=1}
        ]}}}
     Scale Curve=0
    )
    Index 1 Param 2=1
    Index 1 Param 3=&FloatFromCurve(FloatCurve=DataInterface() {propertyValues {Curve {keys [<at 0.5 1 RCIM_Cubic>,<at 1 0 RCIM_Cubic>]}}})
    Index 2 Param 1=&FloatFromCurve(
     FloatCurve=DataInterface() {propertyValues {Curve {keys [{tangentMode=RCTM_User, time=0, value=0}, {tangentMode=RCTM_User, time=1, value=1}]}}}
     Scale Curve=0
    )
    Index 2 Param 2=2
    Index 3 Param 4=30
    Module.Index0Float0Write=true
    Module.Index0Float1Write=true
    Module.Index0Float2Write=true
    Module.Index0Float3Write=true
    Module.Index1Float0Write=false
    Module.Index1Float1Write=true
    Module.Index1Float2Write=true
    Module.Index1Float3Write=true
    Module.Index2Float0Write=true
    Module.Index2Float1Write=true
    Module.Index2Float2Write=true
    Module.Index2Float3Write=true
    Module.Index3Float0Write=true
    Module.Index3Float1Write=true
    Module.Index3Float2Write=true
    Module.Index3Float3Write=true
    Write Parameter Index 1=::NewEnumerator1
    Write Parameter Index 2=::NewEnumerator1
    Write Parameter Index 3=::NewEnumerator1
   )
   &ScaleColor(
    Scale Alpha=&FloatFromCurve(FloatCurve=DataInterface() {propertyValues {Curve {keys [
         {tangentMode=RCTM_User, time=0, value=1}
         {arriveTangent=-0.0405, interpMode=RCIM_Cubic, leaveTangent=-0.0405, tangentMode=RCTM_User, time=0.7009, value=1}
         {arriveTangent=-8.0089, interpMode=RCIM_Cubic, leaveTangent=-8.0089, tangentMode=RCTM_User, time=0.9, value=0}
        ]}}})
    Scale RGB=&VectorFromFloat(Value=&FloatFromCurve(
      FloatCurve=DataInterface() {propertyValues {Curve {keys [
          {arriveTangent=9.0725, interpMode=RCIM_Cubic, time=0, value=0}
          <at 0.5033 2 RCIM_Cubic>
          {interpMode=RCIM_Cubic, leaveTangent=-0.0405, time=0.93, value=0}
         ]}}}
      Scale Curve=&FN_FrontEndFloatModifier(Gameplay=3, Front End=3)
     ))
    ScaleA=true
    ScaleRGB=true
   ) {module=ScaleColor001}
   &DynamicMaterialParameters(
    Index 0 Param 1=1
    Index 0 Param 2=1
    Index 0 Param 3=22
    Index 0 Param 4=0.1
    Module.Index0Float0Write=true
    Module.Index0Float1Write=true
    Module.Index0Float2Write=true
    Module.Index0Float3Write=true
    Module.Index1Float0Write=true
    Module.Index1Float1Write=true
    Module.Index1Float2Write=true
    Module.Index1Float3Write=true
    Module.Index2Float0Write=true
    Module.Index2Float1Write=true
    Module.Index2Float2Write=true
    Module.Index2Float3Write=true
    Write Parameter Index 1=::NewEnumerator1
    Write Parameter Index 2=::NewEnumerator1
   ) {enabled=false, module=DynamicMaterialParameters001}
   &UpdateMeshOrientation(
    Facing Direction=$System.FN_Contrail_Setup.Direction Vector
    Facing Interpolation=false
    Orientation Method=::NewEnumerator1
    Override Quat=false
    Stabilization Mode:ENiagara_OrientVectorCount=::NewEnumerator2
    TransformFacing=false
    TransformSide=false
   ) {enabled=false}
   /SetVariables(Particles.MeshOrientation=&MakeCustomQuaternionFromBool(
     Bool:NiagaraBool=$User.bIsFrontEndPreview
     True Quat:Quat4f=&LerpQuaternion(
      Quaternion A:Quat4f=<0 1 0.2 1>
      Quaternion B:Quat4f=&MakeQuaternion(Quaternion From=2, Axis=<1 0 0>, Angle=0, Y=-45, Z=90, From=?, To=?, Angle Type=::NewEnumerator0, Normalize Vectors=false)
      Lerp Factor:NiagaraFloat=$System.FN_Contrail_Setup.Dive Value
      Enable Spherical Linear Interpolation:NiagaraBool=false
     )
     False Quat:Quat4f=<0 1 0 1>
    )) {isSetParameters=true}
   &AlignSpriteToMeshOrientation(Mesh Orientation Relative Sprite Alignment Vector=<-1 0 0>, SetAlignment=true, SetFacing=false)
   /SetVariables(
    Particles.GlowColor=&MultiplyLinearColor(
     Linear Color=<rgba 0.25 0 0.1074 50>
     Scale Factor=&MakeVector4(
      X=1
      Y=1
      Z=1
      W=&FloatFromCurve(FloatCurve=DataInterface() {propertyValues {Curve {keys [
           {arriveTangent=9.0725, interpMode=RCIM_Cubic, time=0, value=0}
           {arriveTangent=2.4467, interpMode=RCIM_Cubic, leaveTangent=2.4467, time=0.5033, value=2}
           {arriveTangent=-1.1256, interpMode=RCIM_Cubic, leaveTangent=-1.1256, tangentMode=RCTM_User, time=0.8006, value=1.9588}
           {interpMode=RCIM_Cubic, leaveTangent=-0.0405, time=0.9112, value=0}
          ]}}})
     )
    )
    Particles.GlowDynamic:Vector4f=&MakeVector4(
     X=&FloatFromCurve(FloatCurve=DataInterface() {propertyValues {Curve {keys [<at 0 1>,<at 0.75 0.5>]}}}, Scale Curve=0.5)
     Y=0
     Z=$Particles.UVy_Pos
     W=0.5
    )
   ) {isSetParameters=true, module=SetVariables_#14}
   /SetVariables(
    Particles.GlowColor2:LinearColor=&MultiplyLinearColor(
     Linear Color=<rgba 0.8958 0 0.3847 1>
     Scale Factor=&MakeVector4(
      X=1
      Y=1
      Z=1
      W=&FloatFromCurve(FloatCurve=DataInterface() {propertyValues {Curve {keys [
           {arriveTangent=9.0725, interpMode=RCIM_Cubic, time=0, value=0}
           {arriveTangent=2.4467, interpMode=RCIM_Cubic, leaveTangent=2.4467, time=0.5033, value=2}
           {arriveTangent=-1.1256, interpMode=RCIM_Cubic, leaveTangent=-1.1256, tangentMode=RCTM_User, time=0.8006, value=1.9588}
           {interpMode=RCIM_Cubic, leaveTangent=-0.0405, time=0.9112, value=0}
          ]}}})
     )
    )
    Particles.GlowDynamic2:Vector4f=&MakeVector4(
     X=&FloatFromCurve(FloatCurve=DataInterface() {propertyValues {Curve {keys [<at 0 1>,<at 0.75 0.5>]}}}, Scale Curve=0.6)
     Y=0
     Z=$Particles.UVy_Pos
     W=0.5
    )
   ) {isSetParameters=true, module=SetVariables_#15}
   /SetVariables(
    Particles.Color=&Multiply_LinearColorByFloat(
     Linear Color=$Particles.Color
     Float=&FNTimeOfDayScalar(Night Scale=0.15, Morning Scale=0.25, Day Scale=0.3, Evening Scale=0.25)
    )
    Particles.GlowColor=&MultiplyLinearColor(
     Linear Color=$Particles.GlowColor
     Scale Factor=&MakeVector4(
      X=1
      Y=1
      Z=1
      W=&FNTimeOfDayScalar(Night Scale=0.25, Morning Scale=0.25, Day Scale=0.25, Evening Scale=0.25)
     )
    )
   ) {isSetParameters=true, module=SetVariables_#16}
   &CameraOffset(Camera Offset Amount=1)
   /SetVariables(Particles.GlowOffset=0.5) {isSetParameters=true, module=SetVariables_#17}
  ]
  rendererClasses [&NiagaraSpriteRendererProperties]
  renderers [
   {rendererClass=&NiagaraSpriteRendererProperties}
   {rendererClass=&NiagaraSpriteRendererProperties, rendererIndex=1}
   {rendererClass=&NiagaraSpriteRendererProperties, rendererIndex=2}
  ]
 }
 ElectricitySprite_Feet {
  emitterUpdateScript [
   &EmitterState()
   &SpawnRate(SpawnRate=8, Use Spawn Probability=false)
   &SpawnBurst_Instantaneous(Spawn Count=4, Spawn Probability=HlslExpression() {hlslExpression="Emitter.LoopCount == 0"}, Use Spawn Probability=true) {enabled=false}
  ]
  particleSpawnScript [
   &PartitionParticles()
   &InitializeParticle(
    Color=<rgba 5 0.651 2.8379 1>
    Lifetime=0.5
    Sprite Size=&SelectVector2DFromArray(
     Array Sampling Mode=::NewEnumerator0
     Direct Array Index=?
     Vector 2D Selection Array=DataInterface() {propertyValues {FloatData [<15 100>, <-15 100>]}}
     Random Seed=0
    )
    Sprite Size Mode=::NewEnumerator4
    UsePositionOffset=false
    Write Material Random=true
   )
   /SetVariables(Particles.UVy_Pos:NiagaraFloat=&RandomRangeFloat(Minimum=0, Maximum=22)) {isSetParameters=true, module=SetVariables_#13}
  ]
  particleUpdateScript [
   &ParticleState()
   &ScaleSpriteSize(
    Scale Factor=&Lerp_Vector2(
     A=&SelectVector2DFromArray(
      Direct Array Index=$Particles.PartitionParticles.Partition
      Vector 2D Selection Array=DataInterface() {propertyValues {FloatData [<1 0.85>, <1 0.85>, <1 1>, <1 1>]}}
     )
     B=&SelectVector2DFromArray(
      Direct Array Index=$Particles.PartitionParticles.Partition
      Vector 2D Selection Array=DataInterface() {propertyValues {FloatData [<1 1.25>, <1 1.25>, <1 1.25>, <1 1.25>]}}
     )
     Alpha=$System.FN_Contrail_Setup.Dive Value
    )
    Scale Sprite Size Mode=::NewEnumerator2
   )
   &SkeletalMeshLocation(
    Bone / Socket Index=&Add_Integer(A=$Particles.PartitionParticles.Partition, B=2)
    MaskSpawnGroup=false
    Mesh Sampling Type=::NewEnumerator5
    Orientation Sampling=::NewEnumerator1
    Sampled Position Offset=&SelectVectorFromArray(
     Direct Array Index=&Add_Integer(A=$Particles.PartitionParticles.Partition, B=2)
     Vector Selection Array=DataInterface() {propertyValues {FloatData [<10 0 5>, <10 0 -5>, <10 0 0>, <10 0 0>]}}
    )
    Skeletal Mesh=$System.Skeletal Mesh
    Socket Sampling Mode=::NewEnumerator5
   )
   &DynamicMaterialParameters(
    Index 0 Param 1=&FloatFromCurve(FloatCurve=DataInterface() {propertyValues {Curve {keys [<at 0 1>,<at 1 0.5>]}}}, Scale Curve=0.4)
    Index 0 Param 3=$Particles.UVy_Pos
    Index 0 Param 4=0.5
    Index 1 Param 1=&FloatFromCurve(
     FloatCurve=DataInterface() {propertyValues {Curve {keys [
         <at 0 0>
         {arriveTangent=1, interpMode=RCIM_Cubic, leaveTangent=1, time=0.3, value=0}
         {interpMode=RCIM_Cubic, leaveTangent=-8.0089, time=1, value=1}
        ]}}}
     Scale Curve=0
    )
    Index 1 Param 2=1
    Index 1 Param 3=&FloatFromCurve(FloatCurve=DataInterface() {propertyValues {Curve {keys [<at 0.5 1 RCIM_Cubic>,<at 1 0 RCIM_Cubic>]}}})
    Index 2 Param 1=&FloatFromCurve(
     FloatCurve=DataInterface() {propertyValues {Curve {keys [{tangentMode=RCTM_User, time=0, value=0}, {tangentMode=RCTM_User, time=1, value=1}]}}}
     Scale Curve=0
    )
    Index 2 Param 2=2
    Index 3 Param 4=30
    Module.Index0Float0Write=true
    Module.Index0Float1Write=true
    Module.Index0Float2Write=true
    Module.Index0Float3Write=true
    Module.Index1Float0Write=false
    Module.Index1Float1Write=true
    Module.Index1Float2Write=true
    Module.Index1Float3Write=true
    Module.Index2Float0Write=true
    Module.Index2Float1Write=true
    Module.Index2Float2Write=true
    Module.Index2Float3Write=true
    Module.Index3Float0Write=true
    Module.Index3Float1Write=true
    Module.Index3Float2Write=true
    Module.Index3Float3Write=true
    Write Parameter Index 1=::NewEnumerator1
    Write Parameter Index 2=::NewEnumerator1
    Write Parameter Index 3=::NewEnumerator1
   )
   &ScaleColor(
    Scale Alpha=&FloatFromCurve(FloatCurve=DataInterface() {propertyValues {Curve {keys [
         {tangentMode=RCTM_User, time=0, value=1}
         {arriveTangent=-0.0405, interpMode=RCIM_Cubic, leaveTangent=-0.0405, tangentMode=RCTM_User, time=0.7009, value=1}
         {arriveTangent=-8.0089, interpMode=RCIM_Cubic, leaveTangent=-8.0089, tangentMode=RCTM_User, time=0.9, value=0}
        ]}}})
    Scale RGB=&VectorFromFloat(Value=&FloatFromCurve(
      FloatCurve=DataInterface() {propertyValues {Curve {keys [
          {arriveTangent=9.0725, interpMode=RCIM_Cubic, time=0, value=0}
          <at 0.5033 2 RCIM_Cubic>
          {interpMode=RCIM_Cubic, leaveTangent=-0.0405, time=0.93, value=0}
         ]}}}
      Scale Curve=&FN_FrontEndFloatModifier(Gameplay=3, Front End=3)
     ))
    ScaleA=true
    ScaleRGB=true
   ) {module=ScaleColor001}
   &DynamicMaterialParameters(
    Index 0 Param 1=1
    Index 0 Param 2=1
    Index 0 Param 3=22
    Index 0 Param 4=0.1
    Module.Index0Float0Write=true
    Module.Index0Float1Write=true
    Module.Index0Float2Write=true
    Module.Index0Float3Write=true
    Module.Index1Float0Write=true
    Module.Index1Float1Write=true
    Module.Index1Float2Write=true
    Module.Index1Float3Write=true
    Module.Index2Float0Write=true
    Module.Index2Float1Write=true
    Module.Index2Float2Write=true
    Module.Index2Float3Write=true
    Write Parameter Index 1=::NewEnumerator1
    Write Parameter Index 2=::NewEnumerator1
   ) {enabled=false, module=DynamicMaterialParameters001}
   &UpdateMeshOrientation(
    Facing Direction=$System.FN_Contrail_Setup.Direction Vector
    Facing Interpolation=false
    Orientation Method=::NewEnumerator1
    Override Quat=false
    Stabilization Mode:ENiagara_OrientVectorCount=::NewEnumerator2
    TransformFacing=false
    TransformSide=false
   ) {enabled=false}
   /SetVariables(Particles.MeshOrientation=&MakeCustomQuaternionFromBool(
     Bool:NiagaraBool=$User.bIsFrontEndPreview
     True Quat:Quat4f=&LerpQuaternion(
      Quaternion A:Quat4f=<0 1 0.2 1>
      Quaternion B:Quat4f=&MakeQuaternion(Quaternion From=2, Axis=<1 0 0>, Angle=0, Y=-45, Z=90, From=?, To=?, Angle Type=::NewEnumerator0, Normalize Vectors=false)
      Lerp Factor:NiagaraFloat=$System.FN_Contrail_Setup.Dive Value
      Enable Spherical Linear Interpolation:NiagaraBool=false
     )
     False Quat:Quat4f=<0 1 0 1>
    )) {isSetParameters=true}
   &AlignSpriteToMeshOrientation(Mesh Orientation Relative Sprite Alignment Vector=<-1 0 0>, SetAlignment=true, SetFacing=false)
   /SetVariables(
    Particles.GlowColor=&MultiplyLinearColor(
     Linear Color=<rgba 0.25 0 0.1074 50>
     Scale Factor=&MakeVector4(
      X=1
      Y=1
      Z=1
      W=&FloatFromCurve(FloatCurve=DataInterface() {propertyValues {Curve {keys [
           {arriveTangent=9.0725, interpMode=RCIM_Cubic, time=0, value=0}
           {arriveTangent=2.4467, interpMode=RCIM_Cubic, leaveTangent=2.4467, time=0.5033, value=2}
           {arriveTangent=-1.1256, interpMode=RCIM_Cubic, leaveTangent=-1.1256, tangentMode=RCTM_User, time=0.8006, value=1.9588}
           {interpMode=RCIM_Cubic, leaveTangent=-0.0405, time=0.9112, value=0}
          ]}}})
     )
    )
    Particles.GlowDynamic:Vector4f=&MakeVector4(
     X=&FloatFromCurve(FloatCurve=DataInterface() {propertyValues {Curve {keys [<at 0 1>,<at 0.75 0.5>]}}}, Scale Curve=0.5)
     Y=0
     Z=$Particles.UVy_Pos
     W=0.5
    )
   ) {isSetParameters=true, module=SetVariables_#14}
   /SetVariables(
    Particles.GlowColor2:LinearColor=&MultiplyLinearColor(
     Linear Color=<rgba 0.8958 0 0.3847 1>
     Scale Factor=&MakeVector4(
      X=1
      Y=1
      Z=1
      W=&FloatFromCurve(FloatCurve=DataInterface() {propertyValues {Curve {keys [
           {arriveTangent=9.0725, interpMode=RCIM_Cubic, time=0, value=0}
           {arriveTangent=2.4467, interpMode=RCIM_Cubic, leaveTangent=2.4467, time=0.5033, value=2}
           {arriveTangent=-1.1256, interpMode=RCIM_Cubic, leaveTangent=-1.1256, tangentMode=RCTM_User, time=0.8006, value=1.9588}
           {interpMode=RCIM_Cubic, leaveTangent=-0.0405, time=0.9112, value=0}
          ]}}})
     )
    )
    Particles.GlowDynamic2:Vector4f=&MakeVector4(
     X=&FloatFromCurve(FloatCurve=DataInterface() {propertyValues {Curve {keys [<at 0 1>,<at 0.75 0.5>]}}}, Scale Curve=0.6)
     Y=0
     Z=$Particles.UVy_Pos
     W=0.5
    )
   ) {isSetParameters=true, module=SetVariables_#15}
   /SetVariables(
    Particles.Color=&Multiply_LinearColorByFloat(
     Linear Color=$Particles.Color
     Float=&FNTimeOfDayScalar(Night Scale=0.3, Morning Scale=0.25, Day Scale=0.3, Evening Scale=0.25)
    )
    Particles.GlowColor=&MultiplyLinearColor(
     Linear Color=$Particles.GlowColor
     Scale Factor=&MakeVector4(
      X=1
      Y=1
      Z=1
      W=&FNTimeOfDayScalar(Night Scale=0.25, Morning Scale=0.25, Day Scale=0.25, Evening Scale=0.25)
     )
    )
   ) {isSetParameters=true, module=SetVariables_#16}
   &CameraOffset(Camera Offset Amount=1)
   /SetVariables(Particles.GlowOffset=0.5) {isSetParameters=true, module=SetVariables_#17}
  ]
  rendererClasses [&NiagaraMeshRendererProperties, &NiagaraSpriteRendererProperties]
  renderers [
   {rendererClass=&NiagaraMeshRendererProperties, rendererIndex=0}
   {rendererClass=&NiagaraSpriteRendererProperties, rendererIndex=1}
   {rendererClass=&NiagaraSpriteRendererProperties, rendererIndex=2}
   {rendererClass=&NiagaraSpriteRendererProperties, rendererIndex=3}
  ]
 }
 HandsFeet_Glow {
  emitterUpdateScript [
   &EmitterState()
   &SpawnBurst_Instantaneous(Spawn Count=4, Use Spawn Probability=false, UseLoopCountLimit=true)
   &SpawnBurst_Instantaneous(Spawn Count=4, Spawn Group=1, Use Spawn Probability=false, UseLoopCountLimit=true) {module=SpawnBurst_Instantaneous001}
  ]
  particleSpawnScript [
   &PartitionParticles(Number Of Partitions=4)
   &InitializeParticle(
    Color=&SelectLinearColorFromArray(
     Array Sampling Mode=::NewEnumerator1
     Direct Array Index=$Particles.PartitionParticles.Partition
     Color Selection Array=DataInterface() {propertyValues {ColorData [<rgba 10 0.2083 2.5704 0.5>, <rgba 10 0.2083 2.5704 0.5>, <rgba 7 0.1458 1.7993 0.5>, <rgba 7 0.1458 1.7993 0.5>]}}
     Random Seed=?
    )
    Sprite Size Mode=::NewEnumerator3
    Uniform Sprite Size=&MakeCustomFloatFromBool(Bool=$User.bIsFrontEndPreview, True Float=35, False Float=40)
    UsePositionOffset=false
    Write Material Random=true
   )
   /SetVariables(Particles.VisibilityTag=$Emitter.SpawnGroup) {isSetParameters=true}
  ]
  particleUpdateScript [
   &ParticleState(Kill Particles When Lifetime Has Elapsed=false, Let Infinitely Lived Particles Die When Emitter Deactivates=true)
   &SpriteRotationRate(Rotation Rate=&RandomRangeFloat(Minimum=45, Maximum=-45))
   &SkeletalMeshLocation(
    Bone / Socket Index=&Add_Integer(A=$Particles.PartitionParticles.Partition, B=9)
    Bone Sampling Mode=::NewEnumerator4
    MaskSpawnGroup=false
    Orientation Sampling=::NewEnumerator1
    Sampled Position Offset=&SelectVectorFromArray(
     Direct Array Index=&Add_Integer(A=$Particles.PartitionParticles.Partition, B=0)
     Vector Selection Array=DataInterface() {propertyValues {FloatData [<8 -1 -1>, <-8 1 1>, <-7.5 5 -1>, <7.5 -5 1>]}}
    )
    Skeletal Mesh=$System.Skeletal Mesh
   )
   &UpdateMeshOrientation(
    Facing Direction=$System.FN_Contrail_Setup.Direction Vector
    Facing Interpolation=false
    Orientation Method=::NewEnumerator1
    Override Quat=false
    Stabilization Mode:ENiagara_OrientVectorCount=::NewEnumerator2
    TransformFacing=false
    TransformSide=false
   ) {enabled=false}
   &AlignSpriteToMeshOrientation(Mesh Orientation Relative Sprite Alignment Vector=<-1 0 0>, SetAlignment=true, SetFacing=false) {enabled=false}
   &ScaleColor(
    Scale Alpha=&FloatFromCurve(
     FloatCurve=DataInterface() {propertyValues {Curve {
        defaultValue=0
        keys [
         {arriveTangent=4.1785, interpMode=RCIM_Cubic, leaveTangent=4.1785, tangentMode=RCTM_User, time=0, value=0}
         {arriveTangent=1.1179, interpMode=RCIM_Cubic, leaveTangent=1.1179, tangentMode=RCTM_User, time=0.2042, value=0.6228}
         {arriveTangent=0.0224, interpMode=RCIM_Cubic, leaveTangent=0.0224, tangentMode=RCTM_User, time=1, value=1}
        ]
       }}}
     Scale Curve=&MakeCustomFloatFromBool(Bool=$User.bIsFrontEndPreview, True Float=0.75, False Float=1)
    )
    ScaleA=true
    ScaleRGB=true
   ) {module=ScaleColor001}
   &ScaleColor(Scale Alpha=0.45, ScaleA=true, ScaleRGB=true)
   &CameraOffset(Camera Offset Amount=&SelectFloatFromArray(Direct Array Index=$Particles.PartitionParticles.Partition, Float Selection Array=DataInterface() {propertyValues {FloatData [5,5,12,12]}}))
   &FN_Light_Attributes(
    Color=<rgba 1 0 0.4437 1>
    Create Color Attribute=true
    Create Exponent Attribute=true
    Create Position Attribute=true
    Create Radius Attribute=true
    Exponent=1
    Position:NiagaraPosition=$Particles.SkeletalMeshLocation.SampledPosition
    Radius=75
    Scale Color=5
   )
  ]
  rendererClasses [&NiagaraSpriteRendererProperties, &NiagaraLightRendererProperties]
  renderers [
   {rendererClass=&NiagaraSpriteRendererProperties}
   {rendererClass=&NiagaraSpriteRendererProperties, rendererIndex=1}
   {rendererClass=&NiagaraLightRendererProperties, rendererIndex=2}
  ]
 }
 HandsFeet_Lightning {
  emitterUpdateScript [
   &EmitterState()
   &SpawnRate(SpawnRate=30, Use Spawn Probability=false)
  ]
  particleSpawnScript [
   &PartitionParticles(Number Of Partitions=4)
   &InitializeParticle(
    Color=<rgba 50 0.7521 25.1492 1>
    Lifetime Max=0.2
    Lifetime Min=0.4
    Lifetime Mode=::NewEnumerator1
    Mesh Scale=&RandomRangeVector(Minimum=<0.4 0.4 0.4>, Maximum=<0.6 0.6 0.6>)
    Mesh Scale Mode=::NewEnumerator4
    Sprite Size Mode=::NewEnumerator3
    Uniform Sprite Size=5
    UsePositionOffset=false
    Write Material Random=true
   )
   /SetVariables(Particles.Mesh index:NiagaraInt32=&RandomRangeInt(Minimum=0, Maximum=2)) {isSetParameters=true}
  ]
  particleUpdateScript [
   &ParticleState()
   &ShapeLocation(MaskSpawnGroup=false, Override Seed=false, Shape Origin=$Particles.SkeletalMeshLocation.SampledPosition, Sphere Radius=5, UseOwnerScale=true)
   &SkeletalMeshLocation(
    Bone / Socket Index=$Particles.PartitionParticles.Partition
    MaskSpawnGroup=false
    Mesh Sampling Type=::NewEnumerator5
    Orientation Sampling=::NewEnumerator1
    Sampled Position Offset=&SelectVectorFromArray(
     Direct Array Index=$Particles.PartitionParticles.Partition
     Vector Selection Array=DataInterface() {propertyValues {FloatData [<10 0 0>, <10 0 0>, <15 0 0>, <15 0 0>]}}
    )
    Skeletal Mesh=$System.Skeletal Mesh
    Socket Sampling Mode=::NewEnumerator5
   )
   &AlignSpriteToMeshOrientation(Mesh Orientation Relative Sprite Alignment Vector=<-1 0 0>, SetAlignment=true, SetFacing=false) {enabled=false}
   &ScaleColor(
    Scale Alpha=&Waveform(
     "Waveform [1]"=::NewEnumerator0
     "[1] Frequency"=&RandomRangeFloat(Minimum=5, Maximum=8)
     "[1] Amplitude Min/Max"=<0.5 1>
    )
    ScaleA=true
    ScaleRGB=true
   )
   &CameraOffset(Camera Offset Amount=10)
   /SetVariables(Particles.Glow color:LinearColor=&MakeLinearColorFromVectorAndFloat(
     "Vector (RGB)"=&Multiply_VectorByFloat(Vector=&MakeVectorFromLinearColorRGB(Color=<rgba 0 0 0 1>), Float=1)
     "Float (Alpha)"=&Multiply_Float(
      A=&Multiply_Float(
       A=&FloatFromCurve(
        FloatCurve=DataInterface() {propertyValues {Curve {keys [
            {interpMode=RCIM_Cubic, tangentMode=RCTM_User, time=0, value=0}
            {interpMode=RCIM_Cubic, tangentMode=RCTM_User, time=0.5, value=1}
            {tangentMode=RCTM_User, time=1, value=0}
           ]}}}
        Scale Curve=&FNTimeOfDayScalar(Night Scale=0.7, Morning Scale=0.3, Evening Scale=0.3)
       )
       B=&Multiply_Float(A=&RandomRangeFloat(Minimum=0, Maximum=1, Evaluation Type=::NewEnumerator1), B=1)
      )
      B=1
     )
    )) {isSetParameters=true}
   &DynamicMaterialParameters(
    Index 0 Param 1=&Multiply_Float(
     A=3
     B=&FloatFromCurve(
      FloatCurve=DataInterface() {propertyValues {Curve {keys [
          {arriveTangent=-2.6097, interpMode=RCIM_Cubic, leaveTangent=-2.6097, tangentMode=RCTM_User, time=0, value=1.5}
          {arriveTangent=-0.6794, interpMode=RCIM_Cubic, leaveTangent=-0.6794, tangentMode=RCTM_User, time=1, value=0}
         ]}}}
      Scale Curve=1.5
     )
    )
    Index 0 Param 2=&FloatFromCurve(FloatCurve=DataInterface() {propertyValues {Curve {keys [
         {arriveTangent=-2.6097, interpMode=RCIM_Cubic, leaveTangent=-2.6097, tangentMode=RCTM_User, time=0, value=1.5}
         {arriveTangent=-0.6794, interpMode=RCIM_Cubic, leaveTangent=-0.6794, tangentMode=RCTM_User, time=1, value=0}
        ]}}})
    Index 0 Param 3=1
    Index 0 Param 4=1
    Module.Param0WriteEnabled=true
    Module.Param1WriteEnabled=true
    Module.Param2WriteEnabled=true
    Module.Param3WriteEnabled=true
   )
   &ScaleColor(
    Scale Alpha=&Multiply_Float(
     A=&FloatFromCurve(
      FloatCurve=DataInterface() {propertyValues {Curve {keys [
          {interpMode=RCIM_Cubic, tangentMode=RCTM_User, time=0, value=0}
          {interpMode=RCIM_Cubic, tangentMode=RCTM_User, time=0.5, value=1}
          {tangentMode=RCTM_User, time=1, value=0}
         ]}}}
      Scale Curve=&FNTimeOfDayScalar(Night Scale=0.7, Morning Scale=0.3, Evening Scale=0.3)
     )
     B=&Multiply_Float(A=&RandomRangeFloat(Minimum=0, Maximum=1, Evaluation Type=::NewEnumerator1), B=1)
    )
    ScaleA=true
    ScaleRGB=true
   ) {module=ScaleColor001}
   /SetVariables(Particles.Glow Dy02:Vector4f=&MakeVector4(
     X=&Multiply_Float(
      A=10
      B=&FloatFromCurve(FloatCurve=DataInterface() {propertyValues {Curve {keys [
           {arriveTangent=-2.6097, interpMode=RCIM_Cubic, leaveTangent=-2.6097, tangentMode=RCTM_User, time=0, value=1.5}
           {arriveTangent=-0.6794, interpMode=RCIM_Cubic, leaveTangent=-0.6794, tangentMode=RCTM_User, time=1, value=0}
          ]}}})
     )
     Y=0
     Z=0
     W=0
    )) {isSetParameters=true}
   /SetVariables(Particles.MeshOrientation=&MakeQuaternion(
     Quaternion From=5
     Axis=?
     Angle=?
     Yaw=?
     Pitch=?
     Roll=?
     X=?
     Y=?
     Z=?
     From=?
     To=?
     Angle Type=::NewEnumerator0
     Coordinate Space=?
     Normalize Vectors=?
    )) {isSetParameters=true}
   &CameraOffset(Camera Offset Amount=15) {module=CameraOffset001}
  ]
  rendererClasses [&NiagaraMeshRendererProperties]
  renderers [
   {rendererClass=&NiagaraMeshRendererProperties, rendererIndex=0}
   {rendererClass=&NiagaraMeshRendererProperties, rendererIndex=1}
  ]
 }
 Legs_Lightning {
  bEnabled=false
  emitterUpdateScript [
   &EmitterState()
   &SpawnRate(SpawnRate=15, Use Spawn Probability=false)
   /SetVariables(Emitter.DepthFade=35) {isSetParameters=true, module=SetVariables_#4}
  ]
  particleSpawnScript [
   &InitializeParticle(
    Color=&FNTimeOfDayColor(
     Night Color=<rgba 50 0.7521 25.1492 1>
     Morning Color=<rgba 50 0.7521 25.1492 1>
     Day Color=<rgba 50 0.7521 25.1492 1>
     Evening Color=<rgba 50 0.7521 25.1492 1>
    )
    Lifetime=&RandomRangeFloat(Minimum=0.5, Maximum=1)
    Mesh Scale=&Multiply_VectorByFloat(
     Vector=&SelectVectorFromArray(
      Array Sampling Mode=::NewEnumerator0
      Direct Array Index=?
      Vector Selection Array=DataInterface() {propertyValues {FloatData [<1 -1 1>, <-1 -1 -1>]}}
      Random Seed=0
     )
     Float=&RandomRangeFloat(Minimum=0.85, Maximum=1.25)
    )
    Mesh Scale Mode=::NewEnumerator4
    UsePositionOffset=false
    Write Material Random=false
   )
   /SetVariables(Particles.BoneIndex:NiagaraInt32=&RandomRangeInt(Minimum=5, Maximum=8)) {isSetParameters=true, module=SetVariables_#5}
   /SetVariables(
    Particles.MeshIndex=&RandomRangeInt(Minimum=0, Maximum=5)
    Particles.bIsTravelingBolt:NiagaraBool=&SetBoolByFloatComparison(
     A=&RandomRangeFloat(Minimum=0.5, Maximum=1)
     B=&FN_FrontEndFloatModifier(Gameplay=0.25, Front End=0)
    )
   ) {isSetParameters=true, module=SetVariables_#6}
   /SetVariables(Particles.MaskSize:NiagaraFloat=&MakeCustomFloatFromBool(Bool=$Particles.bIsTravelingBolt, True Float=&RandomRangeFloat(Minimum=0.75, Maximum=1), False Float=2)) {isSetParameters=true, module=SetVariables_#7}
   &ShapeLocation(
    Apply To Particle Position=::NewEnumerator1
    Cylinder Height=&SelectFloatFromArray(Direct Array Index=$Particles.BoneIndex, Float Selection Array=DataInterface() {propertyValues {FloatData [25,25,50,50]}})
    Cylinder Height Midpoint=1
    Cylinder Radius=10
    MaskSpawnGroup=false
    Override Seed=false
    Rotation Axis=&SelectVectorFromArray(
     Direct Array Index=$Particles.BoneIndex
     Vector Selection Array=DataInterface() {propertyValues {FloatData [<-1 0 0>, <1 0 0>, <-1 0 0>, <1 0 0>]}}
     Random Seed=0
    )
    Rotation Mode=::NewEnumerator1
    Shape Origin=$Engine.Emitter.SimulationPosition
    Shape Primitive=::NewEnumerator1
    Surface Only=false
    UseOwnerScale=true
   )
   &SkeletalMeshLocation(
    Bone / Socket Index=$Particles.BoneIndex
    Bone Sampling Mode=::NewEnumerator4
    MaskSpawnGroup=false
    Orientation Sampling=::NewEnumerator1
    Sampled Position Offset:NiagaraPosition=$Particles.ShapeLocation.ShapePosition
    Skeletal Mesh=$System.Skeletal Mesh
   ) {enabled=false, module=SkeletalMeshLocation001}
   &DynamicMaterialParameters(
    Index 0 Param 1=$System.LightningWidth
    Index 0 Param 2=$Particles.MaskSize
    Index 0 Param 4=0.5
    Module.Param0WriteEnabled=true
    Module.Param1WriteEnabled=true
    Module.Param2WriteEnabled=true
    Module.Param3WriteEnabled=true
   )
   &CameraOffset(Camera Offset Amount=10)
  ]
  particleUpdateScript [
   &SolveForcesAndVelocity(Clamp Velocity=false, Limit Acceleration=false)
   &ParticleState()
   &ShapeLocation(
    Apply To Particle Position=::NewEnumerator1
    Cylinder Distribution=::NewEnumerator2
    Cylinder Height=&SelectFloatFromArray(
     Direct Array Index=/Niagara/DynamicInputs/Subtract/Subtract_Integer(A:NiagaraInt32=$Particles.BoneIndex, B:NiagaraInt32=5)
     Float Selection Array=DataInterface() {propertyValues {FloatData [20,20,30,30]}}
    )
    Cylinder Height Midpoint=1
    Cylinder Radius=2
    MaskSpawnGroup=false
    Rotation Axis=&SelectVectorFromArray(
     Direct Array Index=&Subtract_Integer(A:NiagaraInt32=$Particles.BoneIndex, B:NiagaraInt32=5)
     Vector Selection Array=DataInterface() {propertyValues {FloatData [<1 0 0>, <-1 0 0>, <1 0 0>, <-1 0 0>]}}
    )
    Rotation Mode=::NewEnumerator1
    Shape Primitive=::NewEnumerator1
    U Position=&RandomRangeFloat(Minimum=0, Maximum=1)
    UseOwnerScale=true
    V Position=&Lerp_Float(
     A=&FloatFromCurve(
      FloatCurve=DataInterface() {propertyValues {Curve {defaultValue=1, keys [{tangentMode=RCTM_User, time=0, value=1}, {tangentMode=RCTM_User, time=1, value=0}]}}}
      Scale Curve=&RandomRangeFloat(Minimum=1, Maximum=2)
     )
     B=&RandomRangeFloat(Minimum=0, Maximum=1)
     Alpha=&SelectFloatFromArray(
      Array Sampling Mode=::NewEnumerator0
      Direct Array Index=?
      Float Selection Array=DataInterface() {propertyValues {FloatData [0.75,1]}}
      Random Seed=0
     )
    )
   ) {module=ShapeLocation001}
   &SkeletalMeshLocation(
    Bone / Socket Index=$Particles.BoneIndex
    Bone Sampling Mode=::NewEnumerator4
    MaskSpawnGroup=false
    Orientation Sampling=::NewEnumerator1
    Sampled Position Offset:NiagaraPosition=$Particles.ShapeLocation001.ShapePosition
    Skeletal Mesh=$System.Skeletal Mesh
   )
   /SetVariables(
    Particles.MaskIntensity:NiagaraFloat=&MakeCustomFloatFromBool(
     Bool=$Particles.bIsTravelingBolt
     True Float=2
     False Float=&FloatFromCurve(FloatCurve=DataInterface() {propertyValues {Curve {keys [{tangentMode=RCTM_User, time=0, value=0.333}, {tangentMode=RCTM_User, time=1, value=0.666}]}}})
    )
    Particles.MaskPosition:NiagaraFloat=&MakeCustomFloatFromBool(
     Bool=$Particles.bIsTravelingBolt
     True Float=$Particles.NormalizedAge
     False Float=&FloatFromCurve(FloatCurve=DataInterface() {propertyValues {Curve {keys [<at 0 0>,<at 0.333 0.5>]}}}, Scale Curve=$System.LightningWidth)
    )
    Particles.MeshPosition:NiagaraPosition=$Particles.Position
   ) {isSetParameters=true, module=SetVariables_#8}
   &DynamicMaterialParameters(
    Index 0 Param 1=&FloatFromCurve(FloatCurve=DataInterface() {propertyValues {Curve {keys [<at 0.666 1>,<at 1 0>]}}}, Scale Curve=$System.LightningWidth)
    Index 0 Param 2=$Particles.MaskSize
    Index 0 Param 3=$Particles.MaskPosition
    Index 0 Param 4=$Particles.MaskIntensity
    Module.Param0WriteEnabled=true
    Module.Param1WriteEnabled=true
    Module.Param2WriteEnabled=true
    Module.Param3WriteEnabled=true
   ) {module=DynamicMaterialParameters001}
   &ScaleColor(
    Scale RGB=&VectorFromFloat(Value=&Multiply_Float(
      A=&Lerp_Float(
       A=0.133
       B=1
       Alpha=&Waveform("Waveform [1]"=::NewEnumerator0, "[1] Frequency"=5, "[1] Amplitude Min/Max"=<0.5 1>)
      )
      B=1
     ))
    ScaleA=false
    ScaleRGB=true
   )
   &ScaleColor(
    Color Value To Scale=$Particles.Color
    Scale RGB=&Multiply_VectorByFloat(Vector=<1 1 1>, Float=$System.TimeOfDayBrightnessMultiplier)
    ScaleA=true
    ScaleRGB=true
   ) {module=ScaleColor001}
   /SetVariables(
    Particles.CameraOffsetGlow=9.5
    Particles.ColorGlow:LinearColor=<rgba 0.401 0 0.1722 5>
    Particles.DynamicMaterialParameterGlow:Vector4f=&MakeVector4(
     X=&Add_Float(A=$System.LightningWidth, B=6)
     Y=$Particles.MaskSize
     Z=$Particles.MaskPosition
     W=$Particles.MaskIntensity
    )
   ) {isSetParameters=true, module=SetVariables_#9}
   /SetVariables(Particles.MeshOrientation=&MultiplyQuaternion(
     Quaternion A=$Particles.SkeletalMeshLocation.SampledRotation
     Quaternion B=&SelectQuaternionFromArray(
      Array Sampling Mode:ENiagaraArraySamplingMode=::NewEnumerator1
      Direct Array Index:NiagaraInt32=&Subtract_Integer(A:NiagaraInt32=$Particles.BoneIndex, B:NiagaraInt32=5)
      Interpolation:NiagaraFloat=?
      Direct Array Mode:ENiagara_WrapClamp=::NewEnumerator0
      Evaluation Type:ENiagaraRandomnessEvaluation=::NewEnumerator0
      Normalized Interpolation:NiagaraBool=true
      Recalculate Random Each Loop:NiagaraBool=false
      Randomness Mode:ENiagaraRandomnessMode=::NewEnumerator0
      Random Seed:NiagaraInt32=?
      Fixed Random Seed:NiagaraBool=false
      Quaternion Selection Array:NiagaraDataInterfaceArrayQuat=DataInterface() {propertyValues {GpuSyncMode=SyncCpuToGpu, MaxElements=0, QuatData [<0 0 0.7071 0.7071>, <0 0 -0.7071 0.7071>, <0 0 0.7071 0.7071>, <0 0 -0.7071 0.7071>]}}
      Enable Spherical Linear Interpolation:NiagaraBool=?
     )
    )) {isSetParameters=true, module=SetVariables_#10}
  ]
  rendererClasses [&NiagaraMeshRendererProperties]
  renderers [
   {rendererClass=&NiagaraMeshRendererProperties, rendererIndex=0}
   {rendererClass=&NiagaraMeshRendererProperties, rendererIndex=1}
  ]
 }
 Lights {
  bEnabled=false
  emitterUpdateScript [
   &EmitterState()
   &SpawnBurst_Instantaneous(Spawn Count=4, Spawn Group=1, Use Spawn Probability=false, UseLoopCountLimit=true)
  ]
  particleSpawnScript [
   &PartitionParticles(Number Of Partitions=4)
   &InitializeParticle(Color=<rgba 1 0.0283 0.2627 1>, UsePositionOffset=false, Write Material Random=true)
   /SetVariables(Particles.VisibilityTag=$Emitter.SpawnGroup) {isSetParameters=true}
  ]
  particleUpdateScript [
   &ParticleState(Kill Particles When Lifetime Has Elapsed=false, Let Infinitely Lived Particles Die When Emitter Deactivates=true)
   &SkeletalMeshLocation(
    Bone / Socket Index=$Particles.PartitionParticles.Partition
    MaskSpawnGroup=false
    Mesh Sampling Type=::NewEnumerator5
    Orientation Sampling=::NewEnumerator1
    Skeletal Mesh=$System.Skeletal Mesh
    Socket Sampling Mode=::NewEnumerator5
   )
   &AlignSpriteToMeshOrientation(Mesh Orientation Relative Sprite Alignment Vector=<-1 0 0>, SetAlignment=true, SetFacing=false) {enabled=false}
   &FN_Light_Attributes(
    Color=<rgba 1 0 0.4437 1>
    Create Color Attribute=true
    Create Exponent Attribute=true
    Create Position Attribute=true
    Create Radius Attribute=true
    Exponent=1
    Position:NiagaraPosition=$Particles.SkeletalMeshLocation.SampledPosition
    Radius=75
    Scale Color=5
   )
  ]
  rendererClasses [&NiagaraLightRendererProperties]
  renderers [{rendererClass=&NiagaraLightRendererProperties, rendererIndex=0}]
 }
}
summary {
 systemName=NS_Contrail_SlickSwish
 emitters [
  {
   emitterName=HandsFeet_Glow
   rendererClasses [&NiagaraSpriteRendererProperties, &NiagaraLightRendererProperties]
  }
  {emitterName=HandsFeet_Lightning, rendererClasses [&NiagaraMeshRendererProperties]}
  {bEnabled=false, emitterName=Lights, rendererClasses [&NiagaraLightRendererProperties]}
  {bEnabled=false, emitterName=ElectricityOLD, rendererClasses [&NiagaraMeshRendererProperties]}
  {bEnabled=false, emitterName=Arms_Lightning, rendererClasses [&NiagaraMeshRendererProperties]}
  {bEnabled=false, emitterName=Legs_Lightning, rendererClasses [&NiagaraMeshRendererProperties]}
  {emitterName=ElectricityMesh, rendererClasses [&NiagaraMeshRendererProperties]}
  {emitterName=ElectricitySprite, rendererClasses [&NiagaraSpriteRendererProperties]}
  {
   emitterName=ElectricitySprite_Feet
   rendererClasses [&NiagaraMeshRendererProperties, &NiagaraSpriteRendererProperties]
  }
  {emitterName=Arms-Legs_Lightning, rendererClasses [&NiagaraMeshRendererProperties]}
 ]
 userVariables [
  User.bIsFrontEnd {type=NiagaraBool}
  User.bIsFrontEndPreview {type=NiagaraBool}
  User.ParaGlideLeanAlpha {type=NiagaraFloat, defaultValue=0, description=""}
  User.ParticlesAmount {type=NiagaraFloat, defaultValue=1, description=""}
  User.PlayerSpeed {type=NiagaraFloat, defaultValue=0, description=""}
  User.RibbonIDOffset {type=NiagaraInt32}
  User.SkeletalMesh {
   type=NiagaraDataInterfaceSkeletalMesh
   defaultValue {
    struct=NiagaraExt_VariableValue_DataInterface
    value {dataInterface=&self:NiagaraDataInterfaceSkeletalMesh_0}
   }
  }
  User.VariantSwitch {type=NiagaraFloat, defaultValue=0, description=""}
 ]
}
